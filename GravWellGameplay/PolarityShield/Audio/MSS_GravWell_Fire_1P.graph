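# One-shot 5.1 MetaSound for the first-person GravWell fire sound.
# On play, random shot and foley waves start while a delayed-release ADSR shapes audio read from the GravWell audio bus.
# A stereo mixer blends all three into front left/right, while the shot's other channels pass straight through and its end finishes the sound.
# The Mobile page plays only a random mono shot on two channels; the pitch and volume randomization inputs are unconnected.
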MetaSoundSource /GravWellGameplay/PolarityShield/Audio/MSS_GravWell_Fire_1P
inputs [
 Audio Bus 1 {type=AudioBusAsset}
 Foley {type=WaveAsset:Array}
 Random.Pitch.Base {type=Float}
 Random.Pitch.Spread {type=Float}
 Random.Volume.Base {type=Float}
 Random.Volume.Spread {type=Float}
 Shoot {type=WaveAsset:Array}
 Shoot_mobile {type=WaveAsset:Array}
 UE.Source.OnPlay {type=Trigger}
]
interfaces [UE.OutputFormat.5dot1:1.0, UE.Source.OneShot:1.0, UE.Source:1.1]
outputs [
 "UE.OutputFormat.5dot1.Audio:0" {type=Audio}
 "UE.OutputFormat.5dot1.Audio:1" {type=Audio}
 "UE.OutputFormat.5dot1.Audio:2" {type=Audio}
 "UE.OutputFormat.5dot1.Audio:3" {type=Audio}
 "UE.OutputFormat.5dot1.Audio:4" {type=Audio}
 "UE.OutputFormat.5dot1.Audio:5" {type=Audio}
 UE.Source.OneShot.OnFinished {type=Trigger}
]
pages [
 Default {
  is_build_graph=true
  input_defaults {
   Audio Bus 1=&/AB_GravWell_01
   Random.Pitch.Base=0
   Random.Pitch.Spread=1
   Random.Volume.Base=1
   Random.Volume.Spread=1.3
   UE.Source.OnPlay=null
   Foley [
    /GravWellGameplay/PolarityShield/Audio/wav/GravWell_Fire_Foley_1P_01
    /GravWellGameplay/PolarityShield/Audio/wav/GravWell_Fire_Foley_1P_02
    /GravWellGameplay/PolarityShield/Audio/wav/GravWell_Fire_Foley_1P_03
   ]
   Shoot [
    /GravWellGameplay/PolarityShield/Audio/wav/GravWell_Fire_1P_01
    /GravWellGameplay/PolarityShield/Audio/wav/GravWell_Fire_1P_02
    /GravWellGameplay/PolarityShield/Audio/wav/GravWell_Fire_1P_03
   ]
   Shoot_mobile [
    /GravWellGameplay/PolarityShield/Audio/wav/GravWell_Fire_1P_01_mono
    /GravWellGameplay/PolarityShield/Audio/wav/GravWell_Fire_1P_02_mono
    /GravWellGameplay/PolarityShield/Audio/wav/GravWell_Fire_1P_03_mono
   ]
  }
  nodes [
   TriggerDelay_0 = UE.Trigger Delay(In=^UE.Source.OnPlay) {also {defaults {Delay Time=0.2}}}
   ADSREnvelopeADSREnvelope_0 = ADSR Envelope.ADSR Envelope:Audio(Trigger Attack=^UE.Source.OnPlay, Trigger Release=@TriggerDelay_0.Out) {also {defaults {Attack Time=0.4, Decay Time=0.3, Release Time=1.3}}}
   AudioBusReader2_0 = "UE.Audio Bus Reader (2)"(Audio Bus=^Audio Bus 1)
   ArrayRandomGet_1 = Array.Random Get:WaveAsset:Array(In Array=^Shoot, Next=UE.Reroute(Value=^UE.Source.OnPlay).Value)
   WavePlayer_0 = UE.Wave Player:5dot1(Play=@ArrayRandomGet_1.On Next, Wave Asset=@ArrayRandomGet_1.Value)
   ArrayRandomGet_0 = Array.Random Get:WaveAsset:Array(In Array=^Foley, Next=UE.Reroute(Value=^UE.Source.OnPlay).Value)
   WavePlayer_1 = UE.Wave Player:Stereo(Play=@ArrayRandomGet_0.On Next, Wave Asset=@ArrayRandomGet_0.Value)
   AudioMixerAudioMixerStereo3_0 = "AudioMixer.Audio Mixer (Stereo, 3)"(
    In 0 R=UE.Multiply:Audio(AdditionalOperands=@ADSREnvelopeADSREnvelope_0.Out Envelope, PrimaryOperand=@AudioBusReader2_0.Out 1).Out
    In 0 L=UE.Multiply:Audio(AdditionalOperands=@ADSREnvelopeADSREnvelope_0.Out Envelope, PrimaryOperand=@AudioBusReader2_0.Out 0).Out
    In 1 L=@WavePlayer_0.Out Front Left
    In 1 R=@WavePlayer_0.Out Front Right
    In 2 L=@WavePlayer_1.Out Left
    In 2 R=@WavePlayer_1.Out Right
   ) {also {defaults {Gain 0=1.3}}}
  ]
  links [
   {dst="@UE.OutputFormat.5dot1.Audio:0", src=AudioMixerAudioMixerStereo3_0.Out L}
   {dst="@UE.OutputFormat.5dot1.Audio:1", src=AudioMixerAudioMixerStereo3_0.Out R}
   {dst="@UE.Source.OneShot.OnFinished", src=WavePlayer_0.On Finished}
   {dst="@UE.OutputFormat.5dot1.Audio:2", src=WavePlayer_0.Out Front Center}
   {dst="@UE.OutputFormat.5dot1.Audio:3", src=WavePlayer_0.Out Low Frequency}
   {dst="@UE.OutputFormat.5dot1.Audio:4", src=WavePlayer_0.Out Side Left}
   {dst="@UE.OutputFormat.5dot1.Audio:5", src=WavePlayer_0.Out Side Right}
  ]
 }
 Mobile {
  nodes [
   ArrayRandomGet_0 = Array.Random Get:WaveAsset:Array(In Array=^Shoot_mobile, Next=^UE.Source.OnPlay)
   WavePlayer_0 = UE.Wave Player:Mono(Play=@ArrayRandomGet_0.On Next, Wave Asset=@ArrayRandomGet_0.Value)
  ]
  links [
   {dst="@UE.Source.OneShot.OnFinished", src=WavePlayer_0.On Finished}
   {dst="@UE.OutputFormat.5dot1.Audio:0", src=WavePlayer_0.Out Mono}
   {dst="@UE.OutputFormat.5dot1.Audio:1", src=WavePlayer_0.Out Mono}
  ]
 }
]
properties {
 assetImportData=&self:AssetImportData
 attenuationSettings=&/ATT_GravWell_Weapon
 maxDistance=15000
 numChannels=6
 outputFormat=FiveDotOne
 sampleRate=48000
 soundClassObject=&Fort_WeaponRanged
 volume=0.8906
}
