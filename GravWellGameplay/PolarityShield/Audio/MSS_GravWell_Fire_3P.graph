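# Mono one-shot MetaSound for the third-person GravWell fire sound, with close and distant variants.
# On play, two shuffled wave arrays feed separate players, both scaled by a shared random volume floored at 1.
# Attenuation distance crossfades from close to distant, with the distant layer also scaled by DistantMultiplier.
# The close player alone signals completion, and the random pitch inputs are left unconnected.
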MetaSoundSource /GravWellGameplay/PolarityShield/Audio/MSS_GravWell_Fire_3P
inputs [
 3P Shoot Close {type=WaveAsset:Array}
 3P Shoot Dist {type=WaveAsset:Array}
 DistantMultiplier {type=Float}
 Random.Pitch.Base {type=Float}
 Random.Pitch.Spread {type=Float}
 Random.Volume.Base {type=Float}
 Random.Volume.Spread {type=Float}
 UE.Attenuation.Distance {type=Float}
 UE.Source.OnPlay {type=Trigger}
]
interfaces [UE.Attenuation:1.0, UE.OutputFormat.Mono:1.0, UE.Source.OneShot:1.0, UE.Source:1.1]
outputs [
 "UE.OutputFormat.Mono.Audio:0" {type=Audio}
 UE.Source.OneShot.OnFinished {type=Trigger}
]
pages [Default {
  input_defaults {
   DistantMultiplier=0.8
   Random.Pitch.Base=0
   Random.Pitch.Spread=1
   Random.Volume.Base=1
   Random.Volume.Spread=0
   UE.Attenuation.Distance=3840.0002
   UE.Source.OnPlay=null
   3P Shoot Close [
    /GravWellGameplay/PolarityShield/Audio/wav/GravWell_Fire_3P_01
    /GravWellGameplay/PolarityShield/Audio/wav/GravWell_Fire_3P_02
    /GravWellGameplay/PolarityShield/Audio/wav/GravWell_Fire_3P_03
   ]
   3P Shoot Dist [
    /GravWellGameplay/PolarityShield/Audio/wav/GravWell_Fire_3P_Dist_01
    /GravWellGameplay/PolarityShield/Audio/wav/GravWell_Fire_3P_Dist_02
    /GravWellGameplay/PolarityShield/Audio/wav/GravWell_Fire_3P_Dist_03
   ]
  }
  nodes [
   n1 = UE.Reroute(Value=^UE.Source.OnPlay)
   n2 = Max.Max:Float(A=UE.RandomFloat(
     Max=UE.Add:Float(PrimaryOperand=^Random.Volume.Base, AdditionalOperands=^Random.Volume.Spread).Out
     Next=@n1.Value
     Min=UE.Subtract:Float(PrimaryOperand=^Random.Volume.Base, AdditionalOperands=^Random.Volume.Spread).Out
    ).Value) {also {defaults {B=1}}}
   n3 = Array.Shuffle:WaveAsset:Array(In Array=^3P Shoot Dist, Next=UE.Reroute(Value=@n1.Value).Value) {also {defaults {Enable Shared State="<MetaSoundBuilderResult.SUCCEEDED: 0>"}}}
   WavePlayer_0 = UE.Wave Player:Mono(Play=@n3.On Next, Wave Asset=@n3.Value)
   n4 = Array.Shuffle:WaveAsset:Array(In Array=^3P Shoot Close, Next=^UE.Source.OnPlay) {also {defaults {Enable Shared State="<MetaSoundBuilderResult.SUCCEEDED: 0>"}}}
   WavePlayer_1 = UE.Wave Player:Mono(Play=@n4.On Next, Wave Asset=@n4.Value)
   CrossfadeTriggerRouteAudio2_0 = "Crossfade.Trigger Route (Audio, 2):Audio"(
    Crossfade Value=MapRange.MapRange:Float(In=^UE.Attenuation.Distance) {also {defaults {In Range B=10000}}}.Out Value
    In 1=UE.Multiply:Audio by Float(
     AdditionalOperands=^DistantMultiplier
     PrimaryOperand=UE.Multiply:Audio by Float(AdditionalOperands=@n2.Value, PrimaryOperand=@WavePlayer_0.Out Mono).Out
    ).Out
    In 0=UE.Multiply:Audio by Float(AdditionalOperands=@n2.Value, PrimaryOperand=@WavePlayer_1.Out Mono).Out
   )
  ]
  links [
   {dst="@UE.OutputFormat.Mono.Audio:0", src=CrossfadeTriggerRouteAudio2_0.Out}
   {dst="@UE.Source.OneShot.OnFinished", src=WavePlayer_1.On Finished}
  ]
 }]
properties {
 assetImportData=&self:AssetImportData
 attenuationSettings=&/ATT_GravWell_Fire_3P
 maxDistance=33000
 sampleRate=48000
 soundClassObject=&Fort_WeaponRanged
 volume=1.2469
 soundSubmixSends [{
   maxSendDistance=15000
   maxSendLevel=0.3
   minSendDistance=6000
   minSendLevel=0.1
   sendLevel=0.1
   sendLevelControlMethod=Linear
   soundSubmix=&Fortnite_Reverb_Distant
  }]
}
