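# One-shot stereo MetaSound for the GravWell weapon's fire charge-up windup, blending first-person and third-person layers by listener distance.
# On play, shuffled picks from both wave arrays start together: stereo 1P, pitched up and scaled by Volume Multiplier, and mono 3P.
# Attenuation distance drives a crossfade from 1P to a reduced-gain 3P in the mixer.
# OnFinished fires once both players finish.
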MetaSoundSource /GravWellGameplay/PolarityShield/Audio/MSS_GravWell_Fire_ChargeUp
inputs [
 Character.Affiliation {type=Enum:FortAudioAffiliation}
 Character.Health {type=Float}
 Character.UniqueID {type=Int32}
 UE.Attenuation.Distance {type=Float}
 UE.Source.OnPlay {type=Trigger}
 Volume Multiplier {type=Float}
 Windup 1P {type=WaveAsset:Array}
 Windup 3P {type=WaveAsset:Array}
]
interfaces [Fort.Character:1.0, UE.Attenuation:1.0, UE.OutputFormat.Stereo:1.0, UE.Source.OneShot:1.0, UE.Source:1.1]
outputs [
 "UE.OutputFormat.Stereo.Audio:0" {type=Audio}
 "UE.OutputFormat.Stereo.Audio:1" {type=Audio}
 UE.Source.OneShot.OnFinished {type=Trigger}
]
pages [Default {
  input_defaults {
   Character.Affiliation=0
   Character.Health=0
   Character.UniqueID=0
   UE.Attenuation.Distance=8720.0029
   UE.Source.OnPlay=null
   Volume Multiplier=0.4
   Windup 1P [
    /GravWellGameplay/PolarityShield/Audio/wav/GravWell_Fire_Windup_1P_01
    /GravWellGameplay/PolarityShield/Audio/wav/GravWell_Fire_Windup_1P_02
    /GravWellGameplay/PolarityShield/Audio/wav/GravWell_Fire_Windup_1P_03
    /GravWellGameplay/PolarityShield/Audio/wav/GravWell_Fire_Windup_1P_04
   ]
   Windup 3P [
    /GravWellGameplay/PolarityShield/Audio/wav/GravWell_Fire_Windup_3P_01
    /GravWellGameplay/PolarityShield/Audio/wav/GravWell_Fire_Windup_3P_02
    /GravWellGameplay/PolarityShield/Audio/wav/GravWell_Fire_Windup_3P_03
    /GravWellGameplay/PolarityShield/Audio/wav/GravWell_Fire_Windup_3P_04
   ]
  }
  nodes [
   n1 = Array.Shuffle:WaveAsset:Array(Next=^UE.Source.OnPlay, In Array=^Windup 1P) {also {defaults {Enable Shared State="<MetaSoundBuilderResult.SUCCEEDED: 0>"}}}
   WavePlayer_1 = UE.Wave Player:Stereo(Play=@n1.On Next, Wave Asset=@n1.Value) {also {defaults {Pitch Shift=2}}}
   CrossfadeTriggerRouteFloat2_0 = "Crossfade.Trigger Route (Float, 2):Float"(Crossfade Value=MapRange.MapRange:Float(In=^UE.Attenuation.Distance) {also {defaults {In Range B=12000}}}.Out Value) {also {defaults {In 1=1}}}
   n2 = UE.Reroute(Value=@CrossfadeTriggerRouteFloat2_0.Out)
   n3 = Array.Shuffle:WaveAsset:Array(Next=^UE.Source.OnPlay, In Array=^Windup 3P) {also {defaults {Enable Shared State="<MetaSoundBuilderResult.SUCCEEDED: 0>"}}}
   WavePlayer_0 = UE.Wave Player:Mono(Play=@n3.On Next, Wave Asset=@n3.Value)
   AudioMixerAudioMixerStereo2_0 = "AudioMixer.Audio Mixer (Stereo, 2)"(
    In 0 L=UE.Multiply:Audio by Float(AdditionalOperands=^Volume Multiplier, PrimaryOperand=@WavePlayer_1.Out Left).Out
    In 0 R=UE.Multiply:Audio by Float(AdditionalOperands=^Volume Multiplier, PrimaryOperand=@WavePlayer_1.Out Right).Out
    Gain 1=UE.Multiply:Float(PrimaryOperand=@n2.Value) {also {defaults {AdditionalOperands=0.6}}}.Out
    Gain 0=UE.Subtract:Float(AdditionalOperands=@n2.Value) {also {defaults {PrimaryOperand=1}}}.Out
    In 1 L=@WavePlayer_0.Out Mono
    In 1 R=@WavePlayer_0.Out Mono
   )
   TriggerAccumulatorTriggerAccumulate2_0 = "TriggerAccumulator.Trigger Accumulate (2)"(In 1=@WavePlayer_0.On Finished, In 0=@WavePlayer_1.On Finished)
  ]
  links [
   {dst="@UE.OutputFormat.Stereo.Audio:0", src=AudioMixerAudioMixerStereo2_0.Out L}
   {dst="@UE.OutputFormat.Stereo.Audio:1", src=AudioMixerAudioMixerStereo2_0.Out R}
   {dst="@UE.Source.OneShot.OnFinished", src=TriggerAccumulatorTriggerAccumulate2_0.Out}
  ]
 }]
properties {
 assetImportData=&self:AssetImportData
 attenuationSettings=&/ATT_GravWell_Weapon_Charge_Up
 bOverrideConcurrency=true
 maxDistance=12000
 numChannels=2
 outputFormat=Stereo
 sampleRate=48000
 soundClassObject=&Fort_WeaponRanged
 volume=2.7609
 concurrencyOverrides {maxCount=3, retriggerTime=0.25}
}
