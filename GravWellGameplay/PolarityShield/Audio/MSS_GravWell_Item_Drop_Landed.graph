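# One-shot mono MetaSound source for the GravWell (Polarity Shield) item landing after being dropped, using two wave variations.
# It wraps the shared MSS_Weapon_Equip graph, forwarding play trigger, character data, sounds and volume, and returning its audio and finished trigger.
# The outer pitch Spread input drives the inner pitch Base; outer Base and Movement.Speed are unconnected.
# Weapon-foley attenuation, sound class and concurrency govern playback.
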MetaSoundSource /GravWellGameplay/PolarityShield/Audio/MSS_GravWell_Item_Drop_Landed
inputs [
 Character.Affiliation {type=Enum:FortAudioAffiliation}
 Character.Health {type=Float}
 Character.UniqueID {type=Int32}
 Drop Landed {type=WaveAsset:Array}
 MasterVolume {type=Float}
 Movement.Speed {type=Float}
 Random.PitchShift.Base {type=Float}
 Random.PitchShift.Spread {type=Float}
 UE.Source.OnPlay {type=Trigger}
]
interfaces [Fort.Character:1.0, UE.OutputFormat.Mono:1.0, UE.Source.OneShot:1.0, UE.Source:1.1]
outputs [
 "UE.OutputFormat.Mono.Audio:0" {type=Audio}
 UE.Source.OneShot.OnFinished {type=Trigger}
]
pages [Default {
  input_defaults {
   Character.Affiliation=0
   Character.Health=0
   Character.UniqueID=0
   MasterVolume=1
   Movement.Speed=0
   Random.PitchShift.Base=-0.002
   Random.PitchShift.Spread=0.5
   UE.Source.OnPlay=null
   Drop Landed [/GravWellGameplay/PolarityShield/Audio/wav/GravWell_DropLanded_01, /GravWellGameplay/PolarityShield/Audio/wav/GravWell_DropLanded_02]
  }
  nodes [MSS_Weapon_Equip_0 = /FortMetasound/Weapons/MSS_Weapon_Equip(
    Character.Affiliation=^Character.Affiliation
    Character.Health=^Character.Health
    Character.UniqueID=^Character.UniqueID
    Sounds=^Drop Landed
    MasterVolume=^MasterVolume
    Random.PitchShift.Base=^Random.PitchShift.Spread
    UE.Source.OnPlay=^UE.Source.OnPlay
   )]
  links [
   {dst="@UE.OutputFormat.Mono.Audio:0", src=MSS_Weapon_Equip_0.UE.OutputFormat.Mono.Audio:0}
   {dst="@UE.Source.OneShot.OnFinished", src=MSS_Weapon_Equip_0.UE.Source.OneShot.OnFinished}
  ]
 }]
properties {
 assetImportData=&self:AssetImportData
 attenuationSettings=&AS_Fort_WeaponFoley
 maxDistance=2250
 sampleRate=48000
 soundClassObject=&Fort_WeaponFoley
 concurrencySet [&CS_Fort_WeaponFoley]
}
