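# One-shot mono MetaSound for the equip sound of the GravWell item in PolarityShield, set up as owner-limited weapon foley.
# The shared MSS_Weapon_Equip graph plays the equip waves, mixed with a random electrical layer shaped by an AD envelope and boosted in gain.
# OnFinished accumulates both the equip completion and the envelope's end.
# The nested graph's pitch Base is wired from this asset's Spread input, leaving its own Base input unused.
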MetaSoundSource /GravWellGameplay/PolarityShield/Audio/MSS_GravWell_Item_Equip
inputs [
 Character.Affiliation {type=Enum:FortAudioAffiliation}
 Character.Health {type=Float}
 Character.UniqueID {type=Int32}
 Elec layer {type=WaveAsset:Array}
 Equip {type=WaveAsset:Array}
 MasterVolume {type=Float}
 Movement.Speed {type=Float}
 Random.PitchShift.Base {type=Float}
 Random.PitchShift.Spread {type=Float}
 UE.Source.OnPlay {type=Trigger}
]
interfaces [Fort.Character:1.0, UE.OutputFormat.Mono:1.0, UE.Source.OneShot:1.0, UE.Source:1.1]
outputs [
 "UE.OutputFormat.Mono.Audio:0" {type=Audio}
 UE.Source.OneShot.OnFinished {type=Trigger}
]
pages [Default {
  input_defaults {
   Character.Affiliation=0
   Character.Health=0
   Character.UniqueID=0
   MasterVolume=0.868
   Movement.Speed=0
   Random.PitchShift.Base=-0.002
   Random.PitchShift.Spread=0
   UE.Source.OnPlay=null
   Elec layer [
    /GravWellGameplay/PolarityShield/Audio/wav/GravWell_Equip_Elec_OS_01
    /GravWellGameplay/PolarityShield/Audio/wav/GravWell_Equip_Elec_OS_02
    /GravWellGameplay/PolarityShield/Audio/wav/GravWell_Equip_Elec_OS_03
   ]
   Equip [/GravWellGameplay/PolarityShield/Audio/wav/GravWell_Equip_01, /GravWellGameplay/PolarityShield/Audio/wav/GravWell_Equip_02]
  }
  nodes [
   n1 = /FortMetasound/Weapons/MSS_Weapon_Equip(
    Character.Affiliation=^Character.Affiliation
    Character.Health=^Character.Health
    Character.UniqueID=^Character.UniqueID
    Sounds=^Equip
    MasterVolume=^MasterVolume
    Random.PitchShift.Base=^Random.PitchShift.Spread
    UE.Source.OnPlay=^UE.Source.OnPlay
   ) {also {defaults {Random.PitchShift.Spread=0}}}
   ArrayRandomGet_0 = Array.Random Get:WaveAsset:Array(In Array=^Elec layer, Next=^UE.Source.OnPlay)
   ADEnvelopeADEnvelope_0 = AD Envelope.AD Envelope:Audio(Trigger=@ArrayRandomGet_0.On Next) {also {defaults {Attack Time=0.5, Decay Time=3}}}
   WavePlayer_0 = UE.Wave Player:Mono(Play=@ArrayRandomGet_0.On Next, Wave Asset=@ArrayRandomGet_0.Value) {also {defaults {Loop="<MetaSoundBuilderResult.SUCCEEDED: 0>"}}}
   AudioMixerAudioMixerMono2_0 = "AudioMixer.Audio Mixer (Mono, 2)"(
    In 0=@n1."UE.OutputFormat.Mono.Audio:0"
    In 1=UE.Multiply:Audio(PrimaryOperand=@ADEnvelopeADEnvelope_0.Out Envelope, AdditionalOperands=@WavePlayer_0.Out Mono).Out
   ) {also {defaults {Gain 1=3}}}
   TriggerAccumulatorTriggerAccumulate2_0 = "TriggerAccumulator.Trigger Accumulate (2)"(In 1=@ADEnvelopeADEnvelope_0.On Done, In 0=@n1.UE.Source.OneShot.OnFinished)
  ]
  links [
   {dst="@UE.OutputFormat.Mono.Audio:0", src=AudioMixerAudioMixerMono2_0.Out}
   {dst="@UE.Source.OneShot.OnFinished", src=TriggerAccumulatorTriggerAccumulate2_0.Out}
  ]
 }]
properties {
 assetImportData=&self:AssetImportData
 attenuationSettings=&AS_Fort_WeaponFoley
 bOverrideConcurrency=true
 maxDistance=2250
 sampleRate=48000
 soundClassObject=&Fort_WeaponFoley
 volume=0.75
 concurrencyOverrides {
  bLimitToOwner=true
  bVolumeScaleCanRelease=true
  maxCount=5
  resolutionRule=StopOldest
  retriggerTime=0.1
  voiceStealReleaseTime=1
  volumeScale=0.5
  volumeScaleAttackTime=0.3
  volumeScaleReleaseTime=0.1
 }
 concurrencySet [&CS_Fort_WeaponFoley]
}
