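# One-shot stereo MetaSound for the GravWell jump launch in the PolarityShield content, with attenuation and a stop-oldest concurrency override.
# On play it shuffles a first-person launch wave, scaled by a random base-plus-spread volume that defaults to no variation.
# A shuffled electric one-shot follows after a half-second delay.
# A stereo mixer blends both layers, favouring the launch, and OnFinished fires once both players finish.
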MetaSoundSource /GravWellGameplay/PolarityShield/Audio/MSS_GravWell_JumpLaunch
inputs [
 1P Jumplaunch {type=WaveAsset:Array}
 Electric One Shots {type=WaveAsset:Array}
 Random.Volume.Base {type=Float}
 Random.Volume.Spread {type=Float}
 UE.Source.OnPlay {type=Trigger}
]
interfaces [UE.OutputFormat.Stereo:1.0, UE.Source.OneShot:1.0, UE.Source:1.1]
outputs [
 "UE.OutputFormat.Stereo.Audio:0" {type=Audio}
 "UE.OutputFormat.Stereo.Audio:1" {type=Audio}
 UE.Source.OneShot.OnFinished {type=Trigger}
]
pages [Default {
  input_defaults {
   Random.Volume.Base=1
   Random.Volume.Spread=0
   UE.Source.OnPlay=null
   1P Jumplaunch [
    /GravWellGameplay/PolarityShield/Audio/wav/GravWell_JumpLaunch_1P_01
    /GravWellGameplay/PolarityShield/Audio/wav/GravWell_JumpLaunch_1P_02
    /GravWellGameplay/PolarityShield/Audio/wav/GravWell_JumpLaunch_1P_03
   ]
   Electric One Shots [
    /GravWellGameplay/PolarityShield/Audio/wav/GravWell_JumpLaunch_Elec_01
    /GravWellGameplay/PolarityShield/Audio/wav/GravWell_JumpLaunch_Elec_02
    /GravWellGameplay/PolarityShield/Audio/wav/GravWell_JumpLaunch_Elec_03
    /GravWellGameplay/PolarityShield/Audio/wav/GravWell_JumpLaunch_Elec_04
    /GravWellGameplay/PolarityShield/Audio/wav/GravWell_JumpLaunch_Elec_05
    /GravWellGameplay/PolarityShield/Audio/wav/GravWell_JumpLaunch_Elec_06
   ]
  }
  nodes [
   n1 = Max.Max:Float(A=UE.RandomFloat(
     Max=UE.Add:Float(PrimaryOperand=^Random.Volume.Base, AdditionalOperands=^Random.Volume.Spread).Out
     Next=UE.Reroute(Value=^UE.Source.OnPlay).Value
     Min=UE.Subtract:Float(PrimaryOperand=^Random.Volume.Base, AdditionalOperands=^Random.Volume.Spread).Out
    ).Value)
   n2 = Array.Shuffle:WaveAsset:Array(In Array=^1P Jumplaunch, Next=^UE.Source.OnPlay) {also {defaults {Enable Shared State="<MetaSoundBuilderResult.SUCCEEDED: 0>"}}}
   WavePlayer_0 = UE.Wave Player:Stereo(Play=@n2.On Next, Wave Asset=@n2.Value)
   TriggerDelay_0 = UE.Trigger Delay(In=^UE.Source.OnPlay) {also {defaults {Delay Time=0.5}}}
   n3 = Array.Shuffle:WaveAsset:Array(In Array=^Electric One Shots, Next=@TriggerDelay_0.Out) {also {defaults {Enable Shared State="<MetaSoundBuilderResult.SUCCEEDED: 0>"}}}
   WavePlayer_1 = UE.Wave Player:Stereo(Play=@n3.On Next, Wave Asset=@n3.Value)
   AudioMixerAudioMixerStereo2_0 = "AudioMixer.Audio Mixer (Stereo, 2)"(
    In 0 L=UE.Multiply:Audio by Float(AdditionalOperands=@n1.Value, PrimaryOperand=UE.Reroute(Value=@WavePlayer_0.Out Left).Value).Out
    In 0 R=UE.Multiply:Audio by Float(AdditionalOperands=@n1.Value, PrimaryOperand=UE.Reroute(Value=@WavePlayer_0.Out Right).Value).Out
    In 1 R=UE.Reroute(Value=@WavePlayer_1.Out Right).Value
    In 1 L=UE.Reroute(Value=@WavePlayer_1.Out Left).Value
   ) {also {defaults {Gain 0=1.4, Gain 1=0.8}}}
   TriggerAccumulatorTriggerAccumulate2_0 = "TriggerAccumulator.Trigger Accumulate (2)"(In 0=@WavePlayer_0.On Finished, In 1=@WavePlayer_1.On Finished)
  ]
  links [
   {dst="@UE.OutputFormat.Stereo.Audio:0", src=AudioMixerAudioMixerStereo2_0.Out L}
   {dst="@UE.OutputFormat.Stereo.Audio:1", src=AudioMixerAudioMixerStereo2_0.Out R}
   {dst="@UE.Source.OneShot.OnFinished", src=TriggerAccumulatorTriggerAccumulate2_0.Out}
  ]
 }]
properties {
 assetImportData=&self:AssetImportData
 attenuationSettings=&/ATT_GravWell_Weapon_Jump_Launch
 bOverrideConcurrency=true
 maxDistance=14000
 numChannels=2
 outputFormat=Stereo
 sampleRate=48000
 soundClassObject=&Fort_WeaponRanged
 volume=0.8906
 concurrencyOverrides {maxCount=4, resolutionRule=StopOldest, retriggerTime=0.4}
}
