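# Parent MetaSound Source for GravWell music in PolarityShield, playing separate instrumental and vocal stereo stems together.
# On play, two wave players start, and each stem's gain is smoothed toward full through InterpTo.
# Instrumental feeds AudioBus1 and vocal feeds AudioBus2, both exposed as inputs alongside the stem assets.
# The source's stereo outputs are unconnected and base submix and sends are disabled, so audio leaves only through the buses.
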MetaSoundSource /GravWellGameplay/PolarityShield/Audio/MSS_GravWell_Music_Parent
inputs [
 AudioBus1 {type=AudioBusAsset}
 AudioBus2 {type=AudioBusAsset}
 Instrumental Stem {type=WaveAsset}
 UE.Source.OnPlay {type=Trigger}
 Vocal Stem {type=WaveAsset}
]
interfaces [UE.OutputFormat.Stereo:1.0, UE.Source:1.1]
outputs [
 "UE.OutputFormat.Stereo.Audio:0" {type=Audio}
 "UE.OutputFormat.Stereo.Audio:1" {type=Audio}
]
pages [Default {
  input_defaults {
   AudioBus1=&/AB_GravWell_01
   AudioBus2=&/AB_GravWell_02
   Instrumental Stem=/GravWellGameplay/PolarityShield/Audio/wav/GravWell_MX_Stem_Instrumental
   UE.Source.OnPlay=null
   Vocal Stem=/GravWellGameplay/PolarityShield/Audio/wav/GravWell_MX_Stem_VOX
  }
  nodes [
   WavePlayer_0 = UE.Wave Player:Stereo(Play=^UE.Source.OnPlay, Wave Asset=^Vocal Stem) {also {defaults {Loop="<MetaSoundBuilderResult.SUCCEEDED: 0>"}}}
   n1 = UE.InterpTo:Audio(Target=Value.Value:Float(Set=@WavePlayer_0.On Play) {also {defaults {Target Value=1}}}.Output Value) {also {defaults {Interp Time=0.2}}}
   n2 = UE.Reroute(Value=@WavePlayer_0.Out Right)
   n3 = UE.Reroute(Value=UE.Reroute(Value=@WavePlayer_0.Out Left).Value)
   AudioBusWriter2_0 = "UE.Audio Bus Writer (2)"(
    Audio Bus=^AudioBus2
    In 1=UE.Reroute(Value=UE.Multiply:Audio by Float(AdditionalOperands=@n1.Value, PrimaryOperand=@n2.Value).Out).Value
    In 0=UE.Reroute(Value=UE.Multiply:Audio by Float(AdditionalOperands=@n1.Value, PrimaryOperand=@n3.Value).Out).Value
   )
   WavePlayer_1 = UE.Wave Player:Stereo(Wave Asset=^Instrumental Stem, Play=^UE.Source.OnPlay) {also {defaults {Loop="<MetaSoundBuilderResult.SUCCEEDED: 0>"}}}
   n4 = UE.InterpTo:Audio(Target=Value.Value:Float(Set=@WavePlayer_1.On Play) {also {defaults {Init Value=0, Target Value=1}}}.Output Value)
   n5 = UE.Reroute(Value=@WavePlayer_1.Out Right)
   n6 = UE.Reroute(Value=@WavePlayer_1.Out Left)
   AudioBusWriter2_1 = "UE.Audio Bus Writer (2)"(
    Audio Bus=^AudioBus1
    In 1=UE.Reroute(Value=UE.Multiply:Audio by Float(AdditionalOperands=@n4.Value, PrimaryOperand=@n5.Value).Out).Value
    In 0=UE.Reroute(Value=UE.Multiply:Audio by Float(AdditionalOperands=@n4.Value, PrimaryOperand=@n6.Value).Out).Value
   )
   InitVariable.Time()
  ]
 }]
properties {
 assetImportData=&self:AssetImportData
 bEnableBaseSubmix=false
 bEnableSubmixSends=false
 bLooping=true
 duration=10000
 numChannels=2
 outputFormat=Stereo
 sampleRate=48000
 soundClassObject=&Fort_WeaponRanged
 volume=0.625
}
