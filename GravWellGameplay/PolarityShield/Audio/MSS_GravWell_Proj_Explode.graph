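# One-shot stereo MetaSound for the GravWell projectile explosion, blending close and distant layers by listener distance.
# On play, each layer shuffles a wave from its own array into a wave player.
# Mapped attenuation distance drives per-channel crossfades; only the close layer takes random pitch and volume inputs, while the distant layer has fixed gain.
# A trigger accumulator signals finished after both players end.
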MetaSoundSource /GravWellGameplay/PolarityShield/Audio/MSS_GravWell_Proj_Explode
inputs [
 Explode Close {type=WaveAsset:Array}
 Explode_Distant {type=WaveAsset:Array}
 Random.Pitch.Base {type=Float}
 Random.Pitch.Spread {type=Float}
 Random.Volume.Base {type=Float}
 Random.Volume.Spread {type=Float}
 UE.Attenuation.Distance {type=Float}
 UE.Source.OnPlay {type=Trigger}
]
interfaces [UE.Attenuation:1.0, UE.OutputFormat.Stereo:1.0, UE.Source.OneShot:1.0, UE.Source:1.1]
outputs [
 "UE.OutputFormat.Stereo.Audio:0" {type=Audio}
 "UE.OutputFormat.Stereo.Audio:1" {type=Audio}
 UE.Source.OneShot.OnFinished {type=Trigger}
]
pages [Default {
  input_defaults {
   Random.Pitch.Base=0
   Random.Pitch.Spread=1
   Random.Volume.Base=0.4
   Random.Volume.Spread=0
   UE.Attenuation.Distance=480.004
   UE.Source.OnPlay=null
   Explode Close [
    /GravWellGameplay/PolarityShield/Audio/wav/GravWell_Impt_Explo_Close_01
    /GravWellGameplay/PolarityShield/Audio/wav/GravWell_Impt_Explo_Close_02
    /GravWellGameplay/PolarityShield/Audio/wav/GravWell_Impt_Explo_Close_03
   ]
   Explode_Distant [
    /GravWellGameplay/PolarityShield/Audio/wav/GravWell_Impt_Explo_Dist_01
    /GravWellGameplay/PolarityShield/Audio/wav/GravWell_Impt_Explo_Dist_02
    /GravWellGameplay/PolarityShield/Audio/wav/GravWell_Impt_Explo_Dist_03
    /GravWellGameplay/PolarityShield/Audio/wav/GravWell_Impt_Explo_Dist_04
   ]
  }
  nodes [
   n1 = MapRange.MapRange:Float(In=^UE.Attenuation.Distance) {also {defaults {In Range A=2000, In Range B=5000}}}
   n2 = Max.Max:Float(A=UE.RandomFloat(
     Max=UE.Add:Float(PrimaryOperand=^Random.Volume.Base, AdditionalOperands=^Random.Volume.Spread).Out
     Next=UE.Reroute(Value=^UE.Source.OnPlay).Value
     Min=UE.Subtract:Float(PrimaryOperand=^Random.Volume.Base, AdditionalOperands=^Random.Volume.Spread).Out
    ).Value)
   n3 = Array.Shuffle:WaveAsset:Array(In Array=^Explode Close, Next=^UE.Source.OnPlay) {also {defaults {Enable Shared State="<MetaSoundBuilderResult.SUCCEEDED: 0>"}}}
   WavePlayer_0 = UE.Wave Player:Stereo(
    Play=@n3.On Next
    Wave Asset=@n3.Value
    Pitch Shift=UE.RandomFloat(
     Max=UE.Add:Float(PrimaryOperand=^Random.Pitch.Base, AdditionalOperands=^Random.Pitch.Spread).Out
     Next=UE.Reroute(Value=^UE.Source.OnPlay).Value
     Min=UE.Subtract:Float(PrimaryOperand=^Random.Pitch.Base, AdditionalOperands=^Random.Pitch.Spread).Out
    ).Value
   )
   n4 = Array.Shuffle:WaveAsset:Array(In Array=^Explode_Distant, Next=^UE.Source.OnPlay) {also {defaults {Enable Shared State="<MetaSoundBuilderResult.SUCCEEDED: 0>"}}}
   WavePlayer_1 = UE.Wave Player:Stereo(Play=@n4.On Next, Wave Asset=@n4.Value)
   CrossfadeTriggerRouteAudio2_0 = "Crossfade.Trigger Route (Audio, 2):Audio"(
    Crossfade Value=@n1.Out Value
    In 0=UE.Multiply:Audio by Float(AdditionalOperands=@n2.Value, PrimaryOperand=@WavePlayer_0.Out Right).Out
    In 1=UE.Multiply:Audio by Float(PrimaryOperand=@WavePlayer_1.Out Right) {also {defaults {AdditionalOperands=0.4}}}.Out
   )
   CrossfadeTriggerRouteAudio2_1 = "Crossfade.Trigger Route (Audio, 2):Audio"(
    Crossfade Value=@n1.Out Value
    In 0=UE.Multiply:Audio by Float(AdditionalOperands=@n2.Value, PrimaryOperand=@WavePlayer_0.Out Left).Out
    In 1=UE.Multiply:Audio by Float(PrimaryOperand=@WavePlayer_1.Out Left) {also {defaults {AdditionalOperands=0.4}}}.Out
   )
   TriggerAccumulatorTriggerAccumulate2_0 = "TriggerAccumulator.Trigger Accumulate (2)"(In 0=@WavePlayer_0.On Finished, In 1=@WavePlayer_1.On Finished)
  ]
  links [
   {dst="@UE.OutputFormat.Stereo.Audio:1", src=CrossfadeTriggerRouteAudio2_0.Out}
   {dst="@UE.OutputFormat.Stereo.Audio:0", src=CrossfadeTriggerRouteAudio2_1.Out}
   {dst="@UE.Source.OneShot.OnFinished", src=TriggerAccumulatorTriggerAccumulate2_0.Out}
  ]
 }]
properties {
 assetImportData=&self:AssetImportData
 attenuationSettings=&/ATT_GravWell_Damage
 bOverrideConcurrency=true
 maxDistance=9150
 numChannels=2
 outputFormat=Stereo
 sampleRate=48000
 soundClassObject=&Fort_WeaponRanged
 volume=2.5
 concurrencyOverrides {maxCount=5, retriggerTime=0.25}
}
