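# One-shot stereo MetaSound for a GravWell projectile impact on a pawn, in the PolarityShield gameplay audio.
# On play, it shuffles through three player-impact waves, plays one, and scales both channels by a random volume from base and spread inputs.
# Pitch base and spread inputs exist but are not wired.
# The source uses damage attenuation, limits concurrency with a retrigger time, and feeds a music sidechain bus.
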MetaSoundSource /GravWellGameplay/PolarityShield/Audio/MSS_GravWell_Proj_Impt_Pawn
inputs [
 Random.Pitch.Base {type=Float}
 Random.Pitch.Spread {type=Float}
 Random.Volume.Base {type=Float}
 Random.Volume.Spread {type=Float}
 Sounds {type=WaveAsset:Array}
 UE.Source.OnPlay {type=Trigger}
]
interfaces [UE.OutputFormat.Stereo:1.0, UE.Source.OneShot:1.0, UE.Source:1.1]
outputs [
 "UE.OutputFormat.Stereo.Audio:0" {type=Audio}
 "UE.OutputFormat.Stereo.Audio:1" {type=Audio}
 UE.Source.OneShot.OnFinished {type=Trigger}
]
pages [Default {
  input_defaults {
   Random.Pitch.Base=0
   Random.Pitch.Spread=1
   Random.Volume.Base=0.6
   Random.Volume.Spread=0
   UE.Source.OnPlay=null
   Sounds [
    /GravWellGameplay/PolarityShield/Audio/wav/GravWell_Impt_Player_03
    /GravWellGameplay/PolarityShield/Audio/wav/GravWell_Impt_Player_02
    /GravWellGameplay/PolarityShield/Audio/wav/GravWell_Impt_Player_01
   ]
  }
  nodes [
   n1 = Max.Max:Float(A=UE.RandomFloat(
     Max=UE.Add:Float(PrimaryOperand=^Random.Volume.Base, AdditionalOperands=^Random.Volume.Spread).Out
     Next=UE.Reroute(Value=^UE.Source.OnPlay).Value
     Min=UE.Subtract:Float(PrimaryOperand=^Random.Volume.Base, AdditionalOperands=^Random.Volume.Spread).Out
    ).Value) {also {defaults {B=-5}}}
   n2 = Array.Shuffle:WaveAsset:Array(In Array=^Sounds, Next=^UE.Source.OnPlay) {also {defaults {Enable Shared State="<MetaSoundBuilderResult.SUCCEEDED: 0>"}}}
   WavePlayer_0 = UE.Wave Player:Stereo(Play=@n2.On Next, Wave Asset=@n2.Value)
   Multiply_0 = UE.Multiply:Audio by Float(AdditionalOperands=@n1.Value, PrimaryOperand=@WavePlayer_0.Out Left)
   Multiply_1 = UE.Multiply:Audio by Float(AdditionalOperands=@n1.Value, PrimaryOperand=@WavePlayer_0.Out Right)
  ]
  links [
   {dst="@UE.OutputFormat.Stereo.Audio:0", src=Multiply_0.Out}
   {dst="@UE.OutputFormat.Stereo.Audio:1", src=Multiply_1.Out}
   {dst="@UE.Source.OneShot.OnFinished", src=WavePlayer_0.On Finished}
  ]
 }]
properties {
 assetImportData=&self:AssetImportData
 attenuationSettings=&/ATT_GravWell_Damage
 bOverrideConcurrency=true
 maxDistance=9150
 numChannels=2
 outputFormat=Stereo
 sampleRate=48000
 soundClassObject=&Fort_WeaponRanged
 volume=2.5
 busSends [{audioBus=&AB_Sidechain_FortMusic_Vol, sendLevel=2}]
 concurrencyOverrides {maxCount=3, retriggerTime=0.3}
}
