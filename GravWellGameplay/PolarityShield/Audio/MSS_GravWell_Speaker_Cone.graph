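# MetaSound source that acts as a looping stereo speaker for audio bus AB_GravWell_02 in the GravWell PolarityShield gameplay audio.
# A single two-channel Audio Bus Reader feeds both outputs directly, so the asset re-emits bus audio and generates no sound itself.
# The GravWell shield loop cone attenuation shapes it spatially, and it uses the ranged weapon sound class.
# Orientation inputs are exposed but unconnected in the graph.
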MetaSoundSource /GravWellGameplay/PolarityShield/Audio/MSS_GravWell_Speaker_Cone
inputs [
 AudioBus2 {type=AudioBusAsset}
 UE.Source.OnPlay {type=Trigger}
 UE.Source.Orientation.Azimuth {type=Float}
 UE.Source.Orientation.Elevation {type=Float}
]
interfaces [UE.OutputFormat.Stereo:1.0, UE.Source.Orientation:1.0, UE.Source:1.1]
outputs [
 "UE.OutputFormat.Stereo.Audio:0" {type=Audio}
 "UE.OutputFormat.Stereo.Audio:1" {type=Audio}
]
pages [Default {
  input_defaults {
   AudioBus2=&/AB_GravWell_02
   UE.Source.OnPlay=null
   UE.Source.Orientation.Azimuth=0
   UE.Source.Orientation.Elevation=0
  }
  nodes [AudioBusReader2_0 = "UE.Audio Bus Reader (2)"(Audio Bus=^AudioBus2)]
  links [{dst="@UE.OutputFormat.Stereo.Audio:0", src=AudioBusReader2_0.Out 0}, {dst="@UE.OutputFormat.Stereo.Audio:1", src=AudioBusReader2_0.Out 1}]
 }]
properties {
 assetImportData=&self:AssetImportData
 attenuationSettings=&/ATT_GravWell_Shield_Loop_Cone
 bDebug=true
 bLooping=true
 duration=10000
 maxDistance=14000
 numChannels=2
 outputFormat=Stereo
 sampleRate=48000
 soundClassObject=&Fort_WeaponRanged
 volume=0.625
}
