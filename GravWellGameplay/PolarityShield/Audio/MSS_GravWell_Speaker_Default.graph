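# Looping stereo MetaSound source for the GravWell speaker, mixing two audio buses with random projectile by-sounds.
# Source orientation fades bus 2; spatialization azimuth attenuates bus 1, gated by an affiliation check on play.
# By-sounds trigger on a distance threshold, with distance driving their low-pass cutoff and gain.
# A distance-gated delay pitch shift on bus 1 is built but never reaches the output.
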MetaSoundSource /GravWellGameplay/PolarityShield/Audio/MSS_GravWell_Speaker_Default
inputs [
 AudioBus1 {type=AudioBusAsset}
 AudioBus2 {type=AudioBusAsset}
 Azimuth {type=Float}
 Bys {type=WaveAsset:Array}
 Character.Affiliation {type=Enum:FortAudioAffiliation}
 Character.Health {type=Float}
 Character.UniqueID {type=Int32}
 Movement.Speed {type=Float}
 Movement.SpeedAzimuth {type=Float}
 Movement.SpeedElevation {type=Float}
 UE.Attenuation.Distance {type=Float}
 UE.Source.OnPlay {type=Trigger}
 UE.Source.Orientation.Azimuth {type=Float}
 UE.Source.Orientation.Elevation {type=Float}
 UE.Spatialization.Azimuth {type=Float}
 UE.Spatialization.Elevation {type=Float}
 WhizbyThreshold {type=Float}
]
interfaces [
 Fort.Character:1.0
 Fort.Movement:1.0
 UE.Attenuation:1.0
 UE.OutputFormat.Stereo:1.0
 UE.Source.Orientation:1.0
 UE.Source:1.1
 UE.Spatialization:1.0
]
outputs [
 "UE.OutputFormat.Stereo.Audio:0" {type=Audio}
 "UE.OutputFormat.Stereo.Audio:1" {type=Audio}
]
pages [Default {
  input_defaults {
   AudioBus1=&/AB_GravWell_01
   AudioBus2=&/AB_GravWell_02
   Azimuth=-180
   Character.Affiliation=0
   Character.Health=0
   Character.UniqueID=0
   Movement.Speed=0
   Movement.SpeedAzimuth=0
   Movement.SpeedElevation=0
   UE.Attenuation.Distance=168.0005
   UE.Source.OnPlay=null
   UE.Source.Orientation.Azimuth=0
   UE.Source.Orientation.Elevation=0
   UE.Spatialization.Azimuth=0.504
   UE.Spatialization.Elevation=0
   WhizbyThreshold=2350
   Bys [
    /GravWellGameplay/PolarityShield/Audio/wav/GravWell_Proj_By_01
    /GravWellGameplay/PolarityShield/Audio/wav/GravWell_Proj_By_02
    /GravWellGameplay/PolarityShield/Audio/wav/GravWell_Proj_By_03
   ]
  }
  nodes [
   AudioBusReader2_0 = "UE.Audio Bus Reader (2)"(Audio Bus=^AudioBus2)
   AudioBusReader2_1 = "UE.Audio Bus Reader (2)"(Audio Bus=^AudioBus1)
   n1 = Clamp.Clamp:Float(In=AbsoluteValue.Abs:Float(Input=^UE.Source.Orientation.Azimuth).Value) {also {defaults {Max=60, Min=5}}}
   n2 = MapRange.MapRange:Float(In=^UE.Attenuation.Distance, In Range B=^WhizbyThreshold) {also {defaults {Out Range A=20000, Out Range B=1000}}}
   ArrayRandomGet_0 = Array.Random Get:WaveAsset:Array(In Array=^Bys, Next=UE.TriggerOnThreshold:Float(In=^UE.Attenuation.Distance, Threshold=^WhizbyThreshold) {also {defaults {Type=1}}}.Out)
   WavePlayer_0 = UE.Wave Player:Stereo(Play=@ArrayRandomGet_0.On Next, Wave Asset=@ArrayRandomGet_0.Value)
   OnePoleLowPassFilter_0 = UE.One-Pole Low Pass Filter:Audio(Cutoff Frequency=@n2.Out Value, In=@WavePlayer_0.Out Left)
   OnePoleLowPassFilter_1 = UE.One-Pole Low Pass Filter:Audio(Cutoff Frequency=@n2.Out Value, In=@WavePlayer_0.Out Right)
   n3 = MapRange.MapRange:Float(In=AbsoluteValue.Abs:Float(Input=^UE.Spatialization.Azimuth).Value) {also {defaults {In Range B=180}}}
   n4 = &TriggerCompareAffiliation(A=^Character.Affiliation, Compare=^UE.Source.OnPlay)
   TriggerRouteTriggerRouteFloat2_0 = "TriggerRoute.Trigger Route (Float, 2):Float"(Set 1=@n4.False, Set 0=@n4.True) {also {defaults {Value 0=1, Value 1=0}}}
   n5 = VariableAccessor.Float(Variable=VariableMutator.Float(Variable=InitVariable.Float().Variable, Value=@TriggerRouteTriggerRouteFloat2_0.Value).Variable)
   AudioMixerAudioMixerStereo3_0 = "AudioMixer.Audio Mixer (Stereo, 3)"(
    In 0 L=@AudioBusReader2_0.Out 0
    In 0 R=@AudioBusReader2_0.Out 1
    In 1 L=@AudioBusReader2_1.Out 0
    In 1 R=@AudioBusReader2_1.Out 1
    Gain 0=UE.InterpTo:Audio(Target=MapRange.MapRange:Float(In=@n1.Value) {also {defaults {In Range A=5, In Range B=60, Out Range A=1, Out Range B=0}}}.Out Value) {also {defaults {Interp Time=0.4}}}.Value
    Gain 2=MapRange.MapRange:Float(In=^UE.Attenuation.Distance, In Range A=^WhizbyThreshold) {also {defaults {In Range B=0, Out Range A=1, Out Range B=0.6}}}.Out Value
    In 2 L=@OnePoleLowPassFilter_0.Out
    In 2 R=@OnePoleLowPassFilter_1.Out
    Gain 1=UE.Subtract:Float(AdditionalOperands=UE.Multiply:Float(PrimaryOperand=@n3.Out Value, AdditionalOperands=@n5.Value).Out) {also {defaults {PrimaryOperand=1}}}.Out
   )
   n6 = UE.InterpTo:Audio(Target=Value.Value:Float(Set=&MSP_OneshotGate(Value=^UE.Attenuation.Distance) {also {defaults {EnterCondition=2, ExitCondition=3, OneshotThreshold=10000}}}.OnGateEnter) {also {defaults {Target Value=-0.5}}}.Output Value)
   DelayPitchShift_0 = UE.Delay Pitch Shift:Audio(In=@AudioBusReader2_1.Out 0, Pitch Shift=@n6.Value) {also {defaults {Delay Length=20}}}
   DelayPitchShift_1 = UE.Delay Pitch Shift:Audio(In=@AudioBusReader2_1.Out 1, Pitch Shift=@n6.Value) {also {defaults {Delay Length=20}}}
   InitVariable.Bool()
   InitVariable.Float()
   InitVariable.Float()
   InitVariable.Float()
  ]
  links [
   {dst="@UE.OutputFormat.Stereo.Audio:0", src=AudioMixerAudioMixerStereo3_0.Out L}
   {dst="@UE.OutputFormat.Stereo.Audio:1", src=AudioMixerAudioMixerStereo3_0.Out R}
  ]
 }]
properties {
 assetImportData=&self:AssetImportData
 attenuationSettings=&/ATT_GravWell_Music
 bDebug=true
 bLooping=true
 duration=10000
 maxDistance=26500
 numChannels=2
 outputFormat=Stereo
 sampleRate=48000
 soundClassObject=&Fort_WeaponRanged
 volume=0.625
}
