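# One-shot stereo MetaSound source that plays the cooldown-end sound for the GravWell Polarity Shield, named as a UI cue.
# On play, a random get picks from a wave array, currently holding one asset, and feeds a stereo wave player straight to the outputs.
# The player's finish trigger ends the source; the MasterVolume input is unconnected.
# It uses the weapon foley sound class with overridden concurrency limiting voice count and retriggering.
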MetaSoundSource /GravWellGameplay/PolarityShield/Audio/MSS_GravWell_UI_Cooldown_End
inputs [
 Cooldown End {type=WaveAsset:Array}
 MasterVolume {type=Float}
 UE.Source.OnPlay {type=Trigger}
]
interfaces [UE.OutputFormat.Stereo:1.0, UE.Source.OneShot:1.0, UE.Source:1.1]
outputs [
 "UE.OutputFormat.Stereo.Audio:0" {type=Audio}
 "UE.OutputFormat.Stereo.Audio:1" {type=Audio}
 UE.Source.OneShot.OnFinished {type=Trigger}
]
pages [Default {
  input_defaults {MasterVolume=1, UE.Source.OnPlay=null, Cooldown End [/GravWellGameplay/PolarityShield/Audio/wav/GravWell_CoolDownEnd]}
  nodes [
   ArrayRandomGet_0 = Array.Random Get:WaveAsset:Array(In Array=^Cooldown End, Next=^UE.Source.OnPlay) {also {defaults {Weights []}}}
   WavePlayer_0 = UE.Wave Player:Stereo(Play=@ArrayRandomGet_0.On Next, Wave Asset=@ArrayRandomGet_0.Value)
  ]
  links [
   {dst="@UE.Source.OneShot.OnFinished", src=WavePlayer_0.On Finished}
   {dst="@UE.OutputFormat.Stereo.Audio:0", src=WavePlayer_0.Out Left}
   {dst="@UE.OutputFormat.Stereo.Audio:1", src=WavePlayer_0.Out Right}
  ]
 }]
properties {
 assetImportData=&self:AssetImportData
 bOverrideConcurrency=true
 numChannels=2
 outputFormat=Stereo
 sampleRate=48000
 soundClassObject=&Fort_WeaponFoley
 volume=0.35
 concurrencyOverrides {maxCount=3, retriggerTime=0.15}
 concurrencySet [&CS_Fort_WeaponFoley]
}
