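# One-shot mono MetaSound source for the GravWell Polarity Shield, layering a dry-fire sound with a "can't do" UI error ping.
# On play, a random pick from the Cooldown End array (one dry-fire wave by default) starts immediately, and the ping follows after a short trigger delay.
# A two-input mixer sums both players; the dry-fire player alone signals OnFinished, and MasterVolume is unconnected.
# It uses weapon-foley sound class and concurrency settings.
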MetaSoundSource /GravWellGameplay/PolarityShield/Audio/MSS_GravWell_UI_DryFire
inputs [
 Can Not Do {type=WaveAsset}
 Cooldown End {type=WaveAsset:Array}
 MasterVolume {type=Float}
 UE.Source.OnPlay {type=Trigger}
]
interfaces [UE.OutputFormat.Mono:1.0, UE.Source.OneShot:1.0, UE.Source:1.1]
outputs [
 "UE.OutputFormat.Mono.Audio:0" {type=Audio}
 UE.Source.OneShot.OnFinished {type=Trigger}
]
pages [Default {
  input_defaults {
   Can Not Do=/Game/Sounds/Fort_UI_Sounds/Gameplay/UI_Error_Ping_Cant_Do_Default_01
   MasterVolume=1
   UE.Source.OnPlay=null
   Cooldown End [/GravWellGameplay/PolarityShield/Audio/wav/GravWell_DryFire]
  }
  nodes [
   ArrayRandomGet_0 = Array.Random Get:WaveAsset:Array(In Array=^Cooldown End, Next=^UE.Source.OnPlay) {also {defaults {Weights []}}}
   WavePlayer_0 = UE.Wave Player:Mono(Play=@ArrayRandomGet_0.On Next, Wave Asset=@ArrayRandomGet_0.Value)
   TriggerDelay_0 = UE.Trigger Delay(In=^UE.Source.OnPlay) {also {defaults {Delay Time=0.1}}}
   WavePlayer_1 = UE.Wave Player:Mono(Wave Asset=^Can Not Do, Play=@TriggerDelay_0.Out)
   AudioMixerAudioMixerMono2_0 = "AudioMixer.Audio Mixer (Mono, 2)"(In 0=@WavePlayer_0.Out Mono, In 1=@WavePlayer_1.Out Mono) {also {defaults {Gain 1=0.9}}}
  ]
  links [
   {dst="@UE.OutputFormat.Mono.Audio:0", src=AudioMixerAudioMixerMono2_0.Out}
   {dst="@UE.Source.OneShot.OnFinished", src=WavePlayer_0.On Finished}
  ]
 }]
properties {
 assetImportData=&self:AssetImportData
 bOverrideConcurrency=true
 sampleRate=48000
 soundClassObject=&Fort_WeaponFoley
 volume=0.5
 concurrencyOverrides {maxCount=3, retriggerTime=0.2}
 concurrencySet [&CS_Fort_WeaponFoley]
}
