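; Music party controller for the GravWell PolarityShield that starts and stops music on its Audio component.
; Start and stop events fade the component in or out only on the Remote branch of an authority check, so nothing plays where the actor has authority.
; Both fades are short, linear and hard-coded, so the stop event's FadeOutDuration parameter is unused.
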(event EventBeginPlay)

(event EventStartMusicParty
  (Networking|SwitchHasAuthority
    (:Remote
      (bind _audio (Variables|Default|GetAudio)
        (:then
          (Audio|Components|Audio|FadeIn _audio 0.1 1.0 0.0 "Linear" :!class /Script/Engine.AudioComponent))
        (:else)))))

(event EventStopMusicParty
  (params (FadeOutDuration :type Float%20%28double-precision%29))
  (Networking|SwitchHasAuthority
    (:Remote
      (bind _audio (Variables|Default|GetAudio)
        (:then
          (Audio|Components|Audio|FadeOut _audio 0.1 0.0 "Linear" :!class /Script/Engine.AudioComponent))
        (:else)))))

(fn ConstructionScript
  (|CallParentFunction))
