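; Blueprint music party prop component named for GravWell, stored under the PolarityShield gameplay audio folder.
; Its BeginPlay event only calls the parent implementation, and the graph adds no other logic of its own.
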(event EventBeginPlay
  (|CallParentFunction))
