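; Projectile building gameplay actor (BGA) for the GravWell Polarity Shield, a moving shield wave that damages structures and hits pawns.
; The server launches it and runs timed async targeting: blockers halt movement and take effects, while hits trigger a granted explosion ability.
; It expires on lifetime, damage cap, steep slopes, disallowed sliding or tagged or undamageable blockers.
; RepNotifies on active time and blocked state drive movement, deploy timeline, trail fade and looping cues.
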(event Custom|Event_OnIsSlidingChanged
  (params (CurrentIsSliding :type Boolean))
  (Niagara|SetNiagaraVariable(Bool) (Variables|Default|GetNS_GravWell_SoundTrail) "IsSliding" CurrentIsSliding)
  (if (and (< (ScalableFloat|GetValueatLevel (Variables|Default|GetAllowProjectileSliding)) 1.0) CurrentIsSliding)
    (Variables|Default|SetbShieldIsBlocked true)
    (CallFunction|OnShieldExpireEvent)))

(event Custom|OnProjectileHit
  (params (HitComponent :type Primitive%20Component%20Object%20Reference) (OtherActor :type Actor%20Object%20Reference) (OtherComp :type Primitive%20Component%20Object%20Reference) (NormalImpulse :type Vector) (Hit :type Hit%20Result%20Structure%20%28by%20ref%29))
  (bind (_returnvalue_roll _returnvalue_pitch _returnvalue_yaw) (Math|Rotator|BreakRotator (Transformation|GetWorldRotation (Variables|Default|GetProjectileCollision))))
  ; Align the shield to world up if it hits anything, only once and only if we allow sliding
  (if (> (ScalableFloat|GetValueatLevel (Variables|Default|GetAllowProjectileSliding)) 0)
    (Transformation|SetWorldRotation (Variables|Default|GetProjectileCollision) (Math|Rotator|MakeRotator _returnvalue_roll 0.0 _returnvalue_yaw) false false)
    (Collision|UnbindEventfromOnComponentHit (Variables|Default|GetProjectileCollision) (delegate Custom|OnProjectileHit))))

(event Custom|OnSteepSlopeHit_Event
  (params (SlopeHitResult :type Hit%20Result%20Structure%20%28by%20ref%29))
  (Variables|Default|SetbShieldIsBlocked true)
  (CallFunction|OnShieldExpireEvent))

(event EventBeginPlay
  (bind _asfort_pawn (Utilities|Casting|CastToFortPawn (Variables|Actor|GetInstigator)))
  (Variables|Default|SetInstigatorPawn _asfort_pawn)
  (if (GameplayTags|HasMatchingGameplayTag (Utilities|Casting|ToGameplayTagAssetInterface (Variables|Actor|GetInstigator)) "(TagName=\"Athena.Quest.ShieldBubble.WithinBubbleArea\")")
    (Ability|GameplayEffect|AddLooseGameplayTags self (Variables|Default|GetTagsSpawnedInBubble)))
  (EventDispatchers|BindEventtoOnIsSlidingChanged (Variables|Movement|GetProjectileMovementComponentFloor) (delegate Custom|Event_OnIsSlidingChanged))
  (Networking|SwitchHasAuthority
    (:Authority
      (Deflection|BindEventtoOnSteepSlopeWasHit (Variables|Movement|GetProjectileMovementComponentFloor) (delegate Custom|OnSteepSlopeHit_Event))
      (Collision|BindEventtoOnComponentHit (Variables|Default|GetProjectileCollision) (delegate Custom|OnProjectileHit))
      (CallFunction|WaveFiringFeedback)
      (Utilities|SetupTargetingPreset
        (:OutputPin
          (if (> (ScalableFloat|GetValueatLevel (Variables|Default|GetDelayInitializeVelocity)) 0)
            (Utilities|Time|SetTimerforNextTickbyEvent (delegate Custom|InitializeVelocity))
            (else
              (CallFunction|InitializeVelocity))))))))

(event Custom|InitializeVelocity
  (Projectile|SetVelocity (Variables|Movement|GetProjectileMovementComponentFloor) (* (Transformation|GetUpVector (Variables|Default|GetShieldCollision)) (CallFunction|CalculateShieldSpeed))))

(event Custom|OnBeatExplosion_Event
  (bind _returnvalue (Ability|GetAbilitySystemComponent 0))
  (bind _returnvalue_2 (Ability|Advanced|GrantandActivateAbilityonSelf _returnvalue (Variables|Default|GetShieldExplosionAbility) self))
  (if (Ability|GameplayAbility|GetGameplayAbilityfromSpecHandle _returnvalue _returnvalue_2)
    (bind _asga_polarity_shield_shield_explosion (Utilities|Casting|CastToGA_PolarityShield_ShieldExplosion (.ReturnValue (Ability|GameplayAbility|GetGameplayAbilityfromSpecHandle _returnvalue _returnvalue_2))))
    (Class|GAPolarityShieldShieldExplosion|SetPolarityWeaponRef _asga_polarity_shield_shield_explosion (Variables|Default|GetPolarityWeapon))))

(event Custom|DoCoreTargeting
  (bind _returnvalue (GameInstanceSubsystems|GetTargetingSubsystem))
  ; Check if structure or BGA boss was hit.
  (bind (_asynctaskproxy _targetinghandle) (Class|AsyncActionPerformTargeting|PerformTargetingAsyncAction 0 (Variables|Default|GetCoreStructureTargetingPreset))
    (:then
      ; Check if a player or boss pawn was hit.
      (bind (_asynctaskproxy_2 _targetinghandle_2) (Class|AsyncActionPerformTargeting|PerformTargetingAsyncAction 0 (Variables|Default|GetCorePawnTargetingPreset))
        (:Targeted
          (if (CallFunction|IsShieldBlocked (TargetingSystem|TargetingResults|GetTargetingResultsActors _returnvalue _targetinghandle_2))
            (CallFunction|HandlePawnTargetingResults (TargetingSystem|TargetingResults|GetTargetingResultsActors _returnvalue _targetinghandle_2))))))
    (:Targeted
      (CallFunction|HandleBuildingTargetingResults _targetinghandle))))

(event Custom|OnShieldLifetimeExpire_Event
  ; If the wave reached the end of its lifetime, bool ensures a GC is played.
  (Variables|Default|SetbShieldReachedLifetimeExpire true)
  (CallFunction|OnShieldExpireEvent))

(event Custom|OnShieldExpire_Event
  ; If the wave reached the end of its lifetime, bool ensures a GC is played.
  (Networking|SwitchHasAuthority
    (:Authority
      (Variables|Default|SetShieldActiveStartTime -1.0)
      (Utilities|Time|ClearandInvalidateTimerbyHandle (Variables|Default|GetBeatExplosionTimerHandle))
      (Utilities|Time|ClearandInvalidateTimerbyHandle (Variables|Default|GetShieldAboutToExpireTimerHandle))
      (Utilities|Time|ClearandInvalidateTimerbyHandle (Variables|Default|GetCoreTargetingTimerHandle))
      (CallFunction|UpdateStructureDamageAnalyticsStat)
      ; Halt the BGA for X seconds before destroying it.
      (Projectile|SetVelocity (Variables|Movement|GetProjectileMovementComponentFloor) (Variables|Default|GetPreDestroyVelocity))
      (if (Variables|Default|GetbShieldReachedLifetimeExpire)
        (Utilities|Time|SetTimerbyEvent (delegate Custom|TimeoutDelay) (Variables|Default|GetDelayUntilBassWaveDestroy))
        (else
          (Gameplay|SilentDie false))))))

(event Custom|OnShieldAboutToExpire_Event
  (Components|Timeline|ReversefromEnd (Variables|BGA_PolarityShield_ShieldPrj|GetShieldDeploy)))

(event Custom|TimeoutDelay
  ; Halt the BGA for X seconds before destroying it.
  (GameplayCue|ExecuteGameplayCue self (Variables|Default|GetGCShieldLifetimeExpire))
  (Gameplay|SilentDie false))

(event Custom|OnFadeOutProj
  (params (FadeOut :type Boolean))
  (bind _ns_gravwell_soundtrail (Variables|Default|GetNS_GravWell_SoundTrail))
  ; FadeOutProjectileTrail
  (Effects|Components|ParticleSystem|SetBoolParameter _ns_gravwell_soundtrail "FadeOut" FadeOut)
  (if FadeOut
    (Utilities|FlowControl|Delay (Variables|Default|GetDelayUntilBassWaveDestroy))
    (Components|Activation|Deactivate _ns_gravwell_soundtrail)))

(fn CalculateShieldSpeed
  (return (* (ScalableFloat|GetValueatLevel (Variables|Default|GetShieldBaseSpeed)) (ScalableFloat|GetValueatLevel (Variables|Default|GetShieldSpeedMultiplier)))))

(fn CheckForBuildingDamageDealt
  (params (ActorDamaged :type Actor%20Object%20Reference) (InDamageAmount :type Float%20%28double-precision%29) (Context :type String))
  (bind _asbuilding_actor (Utilities|Casting|CastToBuildingActor ActorDamaged))
  ; If dealing excess damage to target, only apply actual damage toward damage cap.
  (bind _output_get (LocalVariables|SetDamageCounted (select (ScalableFloat|GetValueasBool (Variables|Default|GetbUseNewDamageCalculation)) (Math|Float|Min(Float) (Variables|GetInDamageAmount) (Utilities|ClampMinFloat (Health|GetHealth _asbuilding_actor))) (Utilities|ClampMinFloat (Health|GetHealth _asbuilding_actor)))))
  ; If DMG cap reached, destroy the wave.
  (bind _output_get_2 (Variables|Default|SetDamageDealt (+ _output_get (Variables|Default|GetDamageDealt))))
  (if (>= _output_get_2 (ScalableFloat|GetValueatLevel (Variables|Default|GetDamageCap)))
    (CallFunction|OnShieldExpireEvent)))

(fn DoesHitActorCauseExpire
  (params (InActor :type Actor%20Object%20Reference))
  (bind _returnvalue (Fortnite|GameplayTags|HasAnyTagsfromNames (Variables|Default|GetTagsShouldCauseExpire) (GameplayTags|GetOwnedGameplayTags (Ability|GetAbilitySystemComponent InActor) :!class /Script/GameplayTags.BlueprintGameplayTagLibrary)))
  (return (Math|Boolean|ORBoolean (not (Variables|/Script/Engine.Actor|GetbCanBeDamaged InActor)) _returnvalue (and (Variables|Default|GetbDealtAllDamageToBlockers) (ScalableFloat|GetValueasBool (Variables|Default|GetbUseNewDamageCalculation))))))

(fn HandleBuildingTargetingResults
  (params (TargetingHandle :type Targeting%20Request%20Handle%20Structure))
  (bind _returnvalue (CallFunction|IsShieldBlocked (TargetingSystem|TargetingResults|GetTargetingResultsActors (GameInstanceSubsystems|GetTargetingSubsystem) TargetingHandle)))
  (bind _returnvalue_2 (GameInstanceSubsystems|GetTargetingSubsystem))
  (bind _targets (TargetingSystem|TargetingResults|GetTargetingResultsActors _returnvalue_2 TargetingHandle))
  (CallFunction|SetShieldBlocked _returnvalue)
  (if _returnvalue
    (for _array_element _targets
      (if (GameplayTags|HasMatchingGameplayTag (Utilities|Casting|ToGameplayTagAssetInterface _array_element) "(TagName=\"Athena.Item.ShieldBubble.Core\")")
        _
        (elif (CallFunction|DoesHitActorCauseExpire _array_element)
          (Variables|Default|SetbShouldExpire true)
          (elif (ScalableFloat|GetValueasBool (Variables|Default|GetbUseNewDamageCalculation))
            (bind _returnvalue_3 (Utilities|Array|AddUnique (Variables|Default|GetCollisionHitActors) _array_element))
            (if (> _returnvalue_3 -1)
              (CallFunction|CheckforBuildingDamageDealt _array_element (ScalableFloat|GetValueatLevel (Variables|Default|GetCollisionDamage)) "Collision")
              (Fortnite|ApplyGameplayEffectsToActor _array_element (select (ScalableFloat|GetValueasBool (Variables|Default|GetbUseNewDamageCalculation)) (Variables|Default|GetBlockingActorEffects_NEW) (Variables|Default|GetBlockingActorEffects_OLD)) (Transformation|GetActorLocation self) false 0 (Ability|GetAbilitySystemComponent 0) (Class|AthenaFunctionLibrary|MakeSimpleFortEffectContext (Variables|Actor|GetInstigator) self)))
            (else
              _)))))
    (if (Variables|Default|GetbShouldExpire)
      (CallFunction|OnShieldExpireEvent)
      (else
        (bind _returnvalue_4 (Fortnite|GameplayTags|HasAnyTagsfromNames (Variables|Default|GetTagsShouldCauseExpire) (GameplayTags|GetOwnedGameplayTags (Ability|GetAbilitySystemComponent 0) :!class /Script/GameplayTags.BlueprintGameplayTagLibrary)))
        (if _returnvalue_4
          (CallFunction|OnShieldExpireEvent))))))

(fn HandlePawnTargetingResults
  (params (PawnActors :type Array%20of%20Actor%20Object%20References))
  (for _array_element PawnActors
    (bind _output_get (Variables|Default|SetCurrentStruckPawn _array_element))
    (if (!= _output_get (Game|GetInstigator))
      (Default|WavePawnImpactFeedback)
      (if (not (Utilities|Array|ContainsItem (Variables|Default|GetAlreadyHitPawns) (Variables|Default|GetCurrentStruckPawn)))
        (GameplayCue|ExecuteGameplayCue (Variables|Default|GetCurrentStruckPawn) (Variables|Default|GetGCStruckPawn))
        (Utilities|Array|Add (Variables|Default|GetAlreadyHitPawns) (Variables|Default|GetCurrentStruckPawn))
        (CallFunction|OnBeatExplosionEvent))
      (return))))

(fn HandleShieldActiveChanged
  (bind _returnvalue (Utilities|IsShieldActive))
  (bind _returnvalue_2 (not _returnvalue))
  (Development|SetHiddeninGame (Variables|Default|GetShieldCollision) _returnvalue_2 false)
  (CallFunction|OnFadeOutProj _returnvalue_2)
  (Networking|SwitchHasAuthority
    (:Authority
      (Utilities|Time|ClearandInvalidateTimerbyHandle (Variables|Default|GetShieldExpireTimerHandle))
      (if (Utilities|IsShieldActive)
        (bind _returnvalue_3 (Utilities|Time|SetTimerbyFunctionName 0 "OnShieldLifetimeExpire_Event" (ScalableFloat|GetValueatLevel (Variables|Default|GetShieldExpireSeconds))))
        (Variables|Default|SetShieldExpireTimerHandle _returnvalue_3)
        (CallFunction|SetShieldMovementEnabled true)
        (CallFunction|DoCoreTargeting)
        (bind _returnvalue_4 (Utilities|Time|SetTimerbyFunctionName 0 "DoCoreTargeting" (ScalableFloat|GetValueatLevel (Variables|Default|GetCoreTargetingIntervalSeconds)) true))
        (Variables|Default|SetCoreTargetingTimerHandle _returnvalue_4)
        (else
          (CallFunction|SetShieldMovementEnabled false))))
    (:Remote
      (Utilities|Time|ClearandInvalidateTimerbyHandle (Variables|Default|GetShieldAboutToExpireTimerHandle))
      (if (Utilities|IsShieldActive)
        (bind _returnvalue_5 (Utilities|Time|SetTimerbyFunctionName 0 "OnShieldAboutToExpire_Event" (Math|Float|Max(Float) (- (ScalableFloat|GetValueatLevel (Variables|Default|GetShieldExpireSeconds)) (Components|Timeline|GetTimelineLength (Variables|BGA_PolarityShield_ShieldPrj|GetShieldDeploy))))))
        (Variables|Default|SetShieldAboutToExpireTimerHandle _returnvalue_5)
        (Components|Timeline|PlayfromStart (Variables|BGA_PolarityShield_ShieldPrj|GetShieldDeploy))
        (Projectile|AddGameplayCue%28Looping%29Local self (Variables|Default|GetGCShieldActive) (LocalVariables|GetTemp%20Parameters))
        (else
          (Projectile|RemoveGameplayCue%28Looping%29Local self (Variables|Default|GetGCShieldActive) (LocalVariables|GetTemp%20Parameters)))))))

(fn IsShieldBlocked
  (params (TargetingHitActorResults :type Array%20of%20Actor%20Object%20References))
  (return (> (Utilities|Array|Length TargetingHitActorResults) 0)))

(fn OnRep_bShieldIsBlocked
  (CallFunction|SetShieldMovementEnabled (not (Variables|Default|GetbShieldIsBlocked)))
  (CallFunction|SetBlastonIntervalEnabled (Variables|Default|GetbShieldIsBlocked)))

(fn OnRep_ShieldActiveStartTime
  (CallFunction|HandleShieldActiveChanged))

(fn SetBlastOnIntervalEnabled
  (params (Enabled :type Boolean))
  (Networking|SwitchHasAuthority
    (:Authority
      (if Enabled
        (if (not (Utilities|Time|IsValidTimerHandle (Variables|Default|GetBeatExplosionTimerHandle)))
          (CallFunction|OnBeatExplosionEvent)
          (if (ScalableFloat|GetValueasBool (Variables|Default|GetbUseNewDamageCalculation))
            (Utilities|Time|SetTimerbyFunctionName 0 "SetDealtAllDamageToBlockers" (ScalableFloat|GetValueatLevel (Variables|Default|GetSetDamageToBlockersDelaySeconds)))))
        (else
          (Utilities|Time|ClearandInvalidateTimerbyHandle (Variables|Default|GetBeatExplosionTimerHandle)))))))

(fn SetDealtAllDamageToBlockers
  (Variables|Default|SetbDealtAllDamageToBlockers true))

(fn SetPolarityWeaponRef
  (params (PolarityWeaponRef :type Fort%20Weapon%20Object%20Reference%20%28by%20ref%29))
  (Variables|Default|SetPolarityWeapon PolarityWeaponRef))

(fn SetShieldBlocked
  (params (bNewShieldBlocked :type Boolean))
  ; Only set if the bool value is changing
  (if (!= (Variables|Default|GetbShieldIsBlocked) bNewShieldBlocked)
    (Variables|Default|SetbShieldIsBlocked bNewShieldBlocked)
    (Variables|Default|SetbDealtAllDamageToBlockers false)))

(fn SetShieldMovementEnabled
  (params (Enabled :type Boolean))
  (bind _projectilemovementcomponentfloor (Variables|Movement|GetProjectileMovementComponentFloor))
  (if Enabled
    (Components|Tick|SetComponentTickEnabled _projectilemovementcomponentfloor true)
    (else
      (Components|Tick|SetComponentTickEnabled _projectilemovementcomponentfloor false))))

(fn UpdateStructureDamageAnalyticsStat
  (bind _polarityweapon (Variables|Default|GetPolarityWeapon))
  (if (and (> (Variables|Default|GetDamageDealt) 0) (Utilities|IsValid _polarityweapon :!class /Script/Engine.KismetSystemLibrary))
    (bind _asgrav_well_weapon (Utilities|Casting|CastToGravWellWeapon _polarityweapon))
    (Class|GravWellWeapon|UpdateStructureDamageAnalytics _asgrav_well_weapon (Variables|Default|GetDamageDealt))))

(fn ConstructionScript)

(fn WaveFiringFeedback
  (bind _asplayer_pawn_athena (Utilities|Casting|CastToPlayerPawn_Athena (Variables|Actor|GetInstigator)))
  (Game|Feedback|ClientStartCameraShake :self (Fortnite|GetFortPlayerControllerfromActor _asplayer_pawn_athena) :Shake "/Game/Effects/Fort_Effects/BluePrints/CameraShakes/B_Small_Explosion_CameraShake.B_Small_Explosion_CameraShake_C" :Scale 1.0 :PlaySpace "CameraLocal")
  (Game|Feedback|ClientPlayForceFeedback (Fortnite|GetFortPlayerControllerfromActor _asplayer_pawn_athena) "/Game/Weapons/ForceFeedback/FF_Athena_FloppingRabbit_Cast.FF_Athena_FloppingRabbit_Cast" "None" false false false))

(fn WavePawnImpactFeedback)
