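; Cosmetic part modifier for the Polarity Shield gauntlets that reflects weapon charges in materials and idle VFX.
; When applied, it writes game time to custom primitive data for an equip wipe, starts a delayed equalizer refresh and binds to ammo changes.
; Ammo changes switch the equalizer's primitive value between cooldown and available, and feed the particle count.
; Weapon changes cache the shield weapon, or deactivate idle VFX when the cast fails.
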(event EventOnCosmeticPartAppliedPostResIn
  (params (Part :type Custom%20Character%20Part%20Object%20Reference))
  ; Material Wipe for Equip Anim
  (Rendering|Material|SetCustomPrimitiveDataFloat (CustomCharacter|GetSkeletalMeshForPartOnPlayerPawn) 1 (Utilities|Time|GetGameTimeinSeconds))
  ; Maybe a more elegant way to do this? Without the timer the material remains at default until fired.
  (Utilities|Time|SetTimerbyEvent (delegate Custom|CheckEqualizerMaterials) 0.5)
  ; If ammo is used, check how many charges are left.
  (EventDispatchers|BindEventtoOnLocalAmmoChanged (Variables|Default|GetAs%20B%20Weap%20Polarity%20Shield) (delegate Custom|On%20Local%20Ammo%20Changed)))

(event Custom|CheckEqualizerMaterials
  ; Maybe a more elegant way to do this? Without the timer the material remains at default until fired.
  (CallFunction|UpdateEqualizerMaterials))

(event Custom|On%20Local%20Ammo%20Changed
  (params (LocalCount :type Integer) (LocalRemaining :type Integer))
  ; If ammo is used, check how many charges are left.
  (bind _output_get (Variables|Default|SetRemainingCharges LocalCount))
  (CallFunction|UpdateEqualizerMaterials)
  (Effects|Components|ParticleSystem|SetIntParameter (CallFunction|GetIdleVFXComponent) "RemainingAmmo" _output_get))

(event EventOnPawnAndPartChanged
  (|CallParentFunction))

(event EventOnWeaponChanged
  (params (NewWeapon :type Fort%20Weapon%20Object%20Reference) (OldWeapon :type Fort%20Weapon%20Object%20Reference))
  (|CallParentFunction NewWeapon OldWeapon)
  (Utilities|IsValid NewWeapon
    (:Is%20Valid
      (bind _asb_weap_polarity_shield (Utilities|Casting|CastToB_Weap_PolarityShield NewWeapon)
        (:then
          (Variables|Default|SetAs%20B%20Weap%20Polarity%20Shield _asb_weap_polarity_shield))
        (:CastFailed
          (Components|Activation|Deactivate (CallFunction|GetIdleVFXComponent)))))))

(fn UpdateEqualizerMaterials
  ; Set equalizers to custom primitive 1 if waiting for cooldown, 2 if charges are available.
  (if (<= (Variables|Default|GetRemainingCharges) 0)
    (Rendering|Material|SetCustomPrimitiveDataFloat (CustomCharacter|GetSkeletalMeshForPartOnPlayerPawn) 0 1.0)
    (else
      (Rendering|Material|SetCustomPrimitiveDataFloat (CustomCharacter|GetSkeletalMeshForPartOnPlayerPawn) 0 2.0))))

(fn UpdateMaterialParams
  (params (ChargeValue :type Float%20%28double-precision%29) (MaxChargeValue :type Float%20%28double-precision%29))
  (bind _returnvalue (CustomCharacter|GetSkeletalMeshForPartOnPlayerPawn))
  (Rendering|Material|SetScalarParameterValueonMaterials _returnvalue "AC_Charges_Current" (Variables|GetChargeValue))
  (Rendering|Material|SetScalarParameterValueonMaterials _returnvalue "AC_Charges_Max" (Variables|GetMaxChargeValue)))

(fn ConstructionScript
  (|CallParentFunction))
