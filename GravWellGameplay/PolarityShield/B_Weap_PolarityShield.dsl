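; Polarity Shield weapon that plays equip, recharge and empty-press feedback and manages its durability as recharging charges run out.
; On equip it fires a burst cue, watches local magazine ammo for recharge cues, and waits for primary or secondary presses to signal failure when empty.
; TryUseDurability spends scaled durability on authority; at an optional cutoff it leaves minimal durability, trims spare ammo to loaded plus charging pips, and exhausts durability on the last shot.
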(event FortWeapon|EventOnEquippedAndAttached
  (GameplayCue|ExecuteGameplayCueOnActor(Burst) (Variables|Default|GetPlayerPawn) (Variables|Default|GetGCWeaponEquippedTag) (Ability|GameplayCue|MakeGameplayCueParameters 0.0))
  (Variables|Default|SetPreviousLocalMagAmmo (FortWeapon|GetMagazineAmmoCount))
  (EventDispatchers|BindEventtoOnLocalAmmoChanged (EventDispatchers|CreateEvent :!event Custom|LocalAmmoChanged))
  ;[0.04916 0.1312 0.15 0.5] Play SFX if player attempts to use weapon while on cooldown.
  (Ability|Async|WaitGameplayEventToActor (Variables|Default|GetPlayerPawn) "(TagName=\"Event.Primary.Pressed\")"
    (:then
      (Ability|Async|WaitGameplayEventToActor (Variables|Default|GetPlayerPawn) "(TagName=\"Event.Secondary.Pressed\")"
        (:EventReceived
          _)))
    (:EventReceived
      _))
  (if (Math|Boolean|ANDBoolean (== (FortWeapon|GetMagazineAmmoCount) 0) (Pawn|IsLocallyViewed (Variables|Default|GetPlayerPawn)) (FortWeapon|IsEquipped))
    (Audio|PlaySound2D (Variables|Default|GetDryFireSound))
    (bind _asfort_player_pawn (Utilities|Casting|CastToFortPlayerPawn (Variables|Default|GetPlayerPawn)))
    (FailedAction|UnabletoPerformAction _asfort_player_pawn true (Variables|Default|GetFailed%20Reason))))

(event EventBeginPlay
  (Variables|Default|SetPlayerPawn (Game|GetInstigator))
  (EventDispatchers|BindEventtoOnConsumeAmmo (EventDispatchers|CreateEvent :!event Custom|Event_OnConsumeAmmo))
  (bind _outrow (Fortnite|GetWeaponStatsRow (Data|GetWeaponDataWeaponStatHandle)))
  (Variables|Default|SetCachedWeaponBaseDurabilityToUse (.DurabilityPerUse (Utilities|Struct|BreakFortBaseWeaponStats _outrow)))
  (Variables|Default|SetCachedClipSize (.ClipSize (Utilities|Struct|BreakFortBaseWeaponStats _outrow))))

(event EventEndPlay
  (params (EndPlayReason :type EEndPlayReason%20Enum :default "Destroyed"))
  (EventDispatchers|UnbindEventfromOnConsumeAmmo (EventDispatchers|CreateEvent :!event Custom|Event_OnConsumeAmmo)))

(event Custom|Event_OnConsumeAmmo
  (params (Quantity :type Integer))
  (Variables|Default|SetPreviousLocalMagAmmo (FortWeapon|GetMagazineAmmoCount)))

(event Custom|LocalAmmoChanged
  (params (LocalCount :type Integer) (LocalRemaining :type Integer))
  ;[0.15 0.0572 0.1304 0.5] Play GC when cooldown complete
  (if (and (> LocalCount (Variables|Default|GetPreviousLocalMagAmmo)) (FortWeapon|IsEquipped))
    (GameplayCue|ExecuteGameplayCue (Variables|Default|GetPlayerPawn) (Variables|Default|GetGCAmmoRechargedTag))))

(fn CheckForDurabilityCutoff
  (params (DurabilityToUse :type Float%20%28double-precision%29))
  (bind (_bisvalid _returnvalue) (FortWeapon|GetDurability))
  (if _bisvalid
    (return (and (<= (- _returnvalue DurabilityToUse) 0) (> (ScalableFloat|GetValueatLevel (Variables|Default|GetbRespectRemainingAmmo)) 0)))
    (else
      (return false))))

(fn GetData
  (return (Variables|Default|GetLaunchActivated)))

(fn HandleReachedDurabilityCutoff
  (bind (_bisvalid _returnvalue) (FortWeapon|GetDurability))
  ; When we reach the cutoff, remove all ammo except what's in the magazine plus what's recharging. When we have one remaining shot, and it's in the magazine, destroy the weapon. Set durability to 1% when we reach this stage to guarantee any future shots will run this logic.
  (if _bisvalid
    (LocalVariables|SetCurrentDurability _returnvalue)
    (FortWeapon|UseWeaponDurability(Explicit) (- (LocalVariables|GetCurrentDurability) 1.0))
    (if (and (== (FortWeapon|GetMagazineAmmoCount) 1) (== (FortWeapon|GetRemainingAmmo) 1))
      (FortWeapon|UseWeaponDurability(Explicit) (LocalVariables|GetCurrentDurability))
      (else
        (LocalVariables|SetAmmoToRemove (- (FortWeapon|GetRemainingAmmo) (FortWeapon|GetMagazineAmmoCount)))
        ; We add 1 because if the magazine is not currently full, we are recharging a pip, so the 1 represents the charging pip
        (if (< (FortWeapon|GetMagazineAmmoCount) (Variables|Default|GetCachedClipSize))
          (LocalVariables|SetAmmoToRemove (- (LocalVariables|GetAmmoToRemove) 1))
          (else
            _))
        (FortWeapon|Ammo|ConsumeSpareAmmo (LocalVariables|GetAmmoToRemove) false)))))

(fn PrintDebugDurability
  (bind (_bisvalid _returnvalue) (FortWeapon|GetDurability))
  (Networking|SwitchHasAuthority
    (:Authority
      (if (and (FortWeapon|GetMaxDurability) (FortWeapon|GetDurability))
        (Development|PrintString (Utilities|String|Append "Mag: " (Utilities|String|ToString(Integer) (FortWeapon|GetMagazineAmmoCount))) true false "(R=0.000000,G=0.660000,B=1.000000,A=1.000000)" 2.0 "PersistR")
        (Development|PrintString (Utilities|String|Append "Durability: " (Utilities|String|ToString(Float) _returnvalue)) true false "(R=0.226641,G=1.000000,B=0.000000,A=1.000000)" 2.0 "PersistD")
        (Development|PrintString (Utilities|String|Append "Ammo: " (Utilities|String|ToString(Integer) (FortWeapon|GetRemainingAmmo))) true false "(R=0.001370,G=1.000000,B=0.618148,A=1.000000)" 2.0 "PersistA")))))

(fn TryUseDurability
  (params (AbilityDurabilityMultiplier :type Float%20%28double-precision%29))
  (Networking|SwitchHasAuthority
    (:Authority
      (LocalVariables|SetDurabilityToUse (* AbilityDurabilityMultiplier (Variables|Default|GetCachedWeaponBaseDurabilityToUse)))
      (if (CallFunction|CheckforDurabilityCutoff (LocalVariables|GetDurabilityToUse))
        (CallFunction|HandleReachedDurabilityCutoff)
        (else
          (FortWeapon|UseWeaponDurability(Explicit) (LocalVariables|GetDurabilityToUse)))))))

(fn ConstructionScript)
