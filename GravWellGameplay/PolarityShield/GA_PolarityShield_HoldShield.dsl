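; Gameplay ability for the GravWell Polarity Shield weapon that plays airborne or ground fire montages, with timers triggering the firing cue.
; The airborne montage trigger starts client-side knockback while falling and grants fall-damage immunity until falling ends.
; On the server, that trigger spends durability and spawns a camera-aimed shield projectile building actor.
; A scalable setting controls when cost commits, including a fallback at ability end.
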(event Custom|FiredAirborne
  (Camera|PushCameraMode "/GravWellGameplay/Assets/Camera/CameraMode_BassWave_Firing_Airborne.CameraMode_BassWave_Firing_Airborne_C")
  (CallFunction|TriggerFiringCue))

(event Ability|EventActivateAbility
  (Variables|Default|SetbAppliedCost false)
  (if (== (ScalableFloat|GetValueatLevel (Variables|Default|GetbCommitAbilitySetting)) 1.0)
    (Ability|CommitAbility)
    (else
      _))
  (Ability|SendGameplayEvent "(TagName=\"Event.Weapons.Fire\")")
  (bind _asfort_pawn (Utilities|Casting|CastToFortPawn (Ability|GetAvatarActorfromActorInfo)))
  (bind _output_get (Variables|Default|SetFortPawn _asfort_pawn))
  (if (GameplayTags|HasMatchingGameplayTag _output_get "(TagName=\"MovementMode.Falling\")")
    (Ability|Tasks|PlayMontageandWaitforTrigger "None" (Variables|Default|GetBassWaveMontageAirborne)
      (:then
        (Ability|GameplayEffect|AddLooseGameplayTags (Variables|Default|GetFortPawn) (Variables|Default|GetGameplay%20Tags))
        (Utilities|Time|SetTimerbyEvent (delegate Custom|FiredAirborne) (Variables|Default|GetDelayUntilFiringAir)))
      (:Triggered
        (if (Ability|DoesTargetHaveTags (Variables|Default|GetFortPawn) (Variables|Default|GetFallingTag))
          (Ability|HasAuthority
            (:True
              _)
            (:False
              (CallFunction|ActivateinAirKnockback)))
          ; Add fall immunity if player fires while airborne and doesn't already have fall damage immunity.
          (if (not (GameplayTags|HasAnyTags (GameplayTags|GetOwnedGameplayTags (Variables|Default|GetFortPawn) :!class /Script/GameplayTags.BlueprintGameplayTagLibrary) (Variables|Default|GetFallDamageImmunityTags)))
            (bind _returnvalue (Ability|ApplyGameplayEffectToOwner (Variables|Default|GetGE%20Fall%20Damage%20Immunity)))
            (bind _output_get_2 (Variables|Default|SetFallDamageImmunityGEHandle _returnvalue))
            (Ability|Tasks|WaitGameplayTagRemoveFromActor (Variables|Default|GetFortPawn) "(TagName=\"MovementMode.Falling\")"
              (:then
                _)
              (:Removed
                (Ability|RemoveGameplayEffectFromOwnerWithHandle _output_get_2)))))
        (Utilities|IsServer
          (:true
            (CallFunction|HandleUseDurability)
            (if (!= (ScalableFloat|GetValueatLevel (Variables|Default|GetbCommitAbilitySetting)) 1.0)
              (Ability|CommitAbility)
              (Variables|Default|SetbAppliedCost true)
              (else
                _))
            (CallFunction|SpawnShieldBGA)
            (bind _asgrav_well_weapon (Utilities|Casting|CastToGravWellWeapon (Ability|GetCurrentSourceWeapon)))
            (Class|GravWellWeapon|CaptureShotFiredAnalytics _asgrav_well_weapon))))
      (:Cancelled
        _)
      (:Completed
        _))
    (Camera|PopCameraMode)
    (Ability|EndAbility)
    (else
      (Ability|Tasks|PlayMontageandWaitforTrigger "None" (Variables|Default|GetBassWaveMontageGround)
        (:then
          (Utilities|Time|SetTimerbyEvent (delegate Custom|FiredGround) (Variables|Default|GetDelayUntilFiringGround)))
        (:Triggered
          _)
        (:Completed
          (Ability|EndAbility))))))

(event Custom|FiredGround
  (CallFunction|TriggerFiringCue))

(event Custom|TriggerFiringCue
  ; Trigger Firing Cue and a GE to pipe a tag to sim proxies so we can know when we're firing.
  (GameplayCue|ExecuteGameplayCue (Variables|Default|GetFortPawn) (Variables|Default|GetGCFiringWave))
  (bind _returnvalue (Development|GetConsoleVariableBoolValue "Fort.Rollback.CheckForExtraTagsToDetermineIsFiring"))
  (if _returnvalue
    (bind _returnvalue_2 (Ability|ApplyGameplayEffectToOwner "/GravWellGameplay/PolarityShield/GE_FiringGravWell.GE_FiringGravWell_C"))
    (Variables|Default|SetFiringGEHandle _returnvalue_2)))

(event Ability|EventOnEndAbility
  (params (bWasCancelled :type Boolean))
  (Ability|GameplayEffect|RemoveLooseGameplayTags (Variables|Default|GetFortPawn) (Variables|Default|GetGameplay%20Tags))
  (Ability|RemoveGameplayEffectFromOwnerWithHandle (Variables|Default|GetFiringGEHandle))
  (Ability|HasAuthority
    (:True
      ; If we haven't applied cost and HF is enabled
      (if (and (not (Variables|Default|GetbAppliedCost)) (> (ScalableFloat|GetValueatLevel (Variables|Default|GetbCommitAbilitySetting)) 1.0))
        (CallFunction|HandleUseDurability)
        (Ability|CommitAbility)
        (Variables|Default|SetbAppliedCost true)))))

(fn ActivateInAirKnockback
  (bind _returnvalue (CallFunction|GetAvatarasFortPlayerPawn))
  (FortMovement|CancelJumpAbility _returnvalue false)
  (Class|FortMovementModeExtBlueprintLibrary|ActivateRegisteredMovementModeExtension (Variables|Default|GetInAirKnockbackMovementMode) _returnvalue)
  (Ability|ExecuteGameplayCueOnOwner (Variables|Default|GetInAirKnockbackCue)))

(fn GetPolarityWeapon
  (bind _asb_weap_polarity_shield (Utilities|Casting|CastToB_Weap_PolarityShield (Ability|GetCurrentSourceWeapon))
    (:then
      (return _asb_weap_polarity_shield))
    (:CastFailed
      (return 0))))

(fn HandleUseDurability
  (bind _polarityweapon (CallFunction|GetPolarityWeapon))
  (Utilities|IsValid _polarityweapon
    (:Is%20Valid
      (Class|BWeapPolarityShield|TryUseDurability _polarityweapon (ScalableFloat|GetValueasDouble (Variables|Default|GetHF_AbilityDurabilityUseMultiplier))))
    (:Is%20Not%20Valid
      _))
  (return))

(fn CanActivateAbility
  (params (ActorInfo :type Gameplay%20Ability%20Actor%20Info%20Structure) (Handle :type Gameplay%20Ability%20Spec%20Handle%20Structure))
  (return 0 (>= (ScalableFloat|GetValueatLevel (Variables|Default|GetHF_CanActivateAbility)) 1.0)))

(fn SpawnShieldBGA
  (bind _returnvalue (CallFunction|GetAvatarasFortPlayerPawn))
  ; If player is airborne, use Air Z Offset instead
  (if (GameplayTags|HasMatchingGameplayTag (Variables|Default|GetFortPawn) "(TagName=\"MovementMode.Falling\")")
    (Variables|Default|SetProjectileZOffset (Variables|Default|GetProjectileZOffsetAir))
    (else
      (Variables|Default|SetProjectileZOffset (Variables|Default|GetProjectileZOffsetGround))))
  (bind _returnvalue_2 (TargetingSystem|GetPlayerCameraTargetingTransform _returnvalue false (Math|Vector|MakeVector (ScalableFloat|GetValueatLevel (Variables|Default|GetProjectileXOffset)) 0.0 (ScalableFloat|GetValueatLevel (Variables|Default|GetProjectileZOffset)))))
  (bind (_returnvalue_location _returnvalue_rotation) (Math|Transform|BreakTransform _returnvalue_2))
  (bind _returnvalue_3 (Building|SpawnBuildingGameplayActor (Variables|Default|GetShieldBGAClass) (Math|Transform|MakeTransform _returnvalue_location _returnvalue_rotation) _returnvalue true false))
  (bind _asbga_polarity_shield_shield_prj (Utilities|Casting|CastToBGA_PolarityShield_ShieldPrj _returnvalue_3))
  (bind _output_get (Variables|Default|SetBlockingBGA _asbga_polarity_shield_shield_prj))
  (Variables|/GravWellGameplay/PolarityShield/BGA_PolarityShield_ShieldPrj.BGA_PolarityShield_ShieldPrj_C|SetShieldActiveStartTime (Fortnite|Sync|Time|GetServerWorldTimeSeconds) _output_get)
  (Class|BGAPolarityShieldShieldPrj|SetPolarityWeaponRef _asbga_polarity_shield_shield_prj (Ability|GetCurrentSourceWeapon)) ; Set source weapon so explosions can report analytics
)
