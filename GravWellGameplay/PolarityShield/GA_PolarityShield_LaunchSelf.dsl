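; Polarity Shield self-launch ability (Bass Jump), blocked while a jump is active unless multi-jump is allowed.
; Activation adds timed blocking tags and pushes a launch camera; the montage's start sets a ground or aerial re-jump timer.
; The airborne branch immediately spends durability, applies fall immunity, damages obstacles above via capsule trace, and activates the launch movement extension on non-authority.
; Authority watches for landing to end the ability, with a timeout.
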(event Ability|EventActivateAbility
  (bind _polarityweapon (Default|GetPolarityWeapon))
  (bind _montage_to_play (Variables|Default|GetMontage%20to%20Play))
  (bind _asfort_pawn (Utilities|Casting|CastToFortPawn (Ability|GetAvatarActorfromActorInfo)))
  (Variables|Default|SetPlayerPawn _asfort_pawn)
  (bind _bcanbassjump (CallFunction|CanBassJump))
  (if _bcanbassjump
    (Ability|SendGameplayEvent "(TagName=\"Event.Weapons.Fire\")")
    (Ability|GameplayEffect|AddLooseGameplayTags (Variables|Default|GetPlayerPawn) (Variables|Default|GetTagsBlockBassWave))
    (Utilities|Time|SetTimerbyEvent (delegate Custom|RemoveLaunchingTags) (ScalableFloat|GetValueatLevel (Variables|Default|GetDurationLockOutPrimary)))
    (Camera|PushCameraMode "/GravWellGameplay/Assets/Camera/CameraMode_BassWave_SelfLaunch.CameraMode_BassWave_SelfLaunch_C")
    (CallFunction|HasAbilityStarted true)
    (if (GameplayTags|HasMatchingGameplayTag (Variables|Default|GetPlayerPawn) "(TagName=\"MovementMode.Falling\")")
      ; If player is airborne, don't wait for the trigger
      (Ability|Tasks|PlayMontageandWaitforTrigger "None" _montage_to_play
        (:then
          (Variables|Default|SetbIsGroundJump false)
          (CallFunction|PauseUntilNextJump)
          (Utilities|Time|SetTimerbyEvent (delegate Custom|End%20Camera) 2.0)
          (Game|Feedback|ClientStartCameraShake :self (Fortnite|GetFortPlayerControllerfromActor (Variables|Default|GetPlayerPawn)) :Shake (Variables|Default|GetBassJumpCamShake) :Scale 1.0 :PlaySpace "CameraLocal")
          (Ability|ApplyGameplayEffectToOwner (Variables|Default|GetFallImmunityGE))
          (Utilities|IsValid _polarityweapon
            (:Is%20Valid
              (Class|BWeapPolarityShield|TryUseDurability _polarityweapon (ScalableFloat|GetValueatLevel (Variables|Default|GetHF_AbilityDurabilityUseMultiplier))))
            (:Is%20Not%20Valid
              _))
          (Ability|CommitAbility)
          (Ability|ExecuteGameplayCueOnOwner (Variables|Default|GetGCBassJumpLaunch))
          (CallFunction|BassJumpBlastThrough)
          (Ability|HasAuthority
            (:True
              ; Wait until the player has landed before ending the ability- maybe needs some work to make it more robust?
              (bind _asfort_pawn_2 (Utilities|Casting|CastToFortPawn (Ability|GetAvatarActorfromActorInfo)))
              (Ability|Tasks|WaitGameplayTagAddToActor _asfort_pawn_2 "(TagName=\"MovementMode.Falling\")"
                (:Added
                  (Ability|Tasks|WaitGameplayTagRemoveFromActor _asfort_pawn_2 "(TagName=\"MovementMode.Falling\")" true
                    (:then
                      (Utilities|Time|SetTimerbyEvent (delegate Custom|EndAbilityTimeout) 2e+01))
                    (:Removed
                      (Ability|EndAbility))))))
            (:False
              (Class|FortMovementModeExtBlueprintLibrary|ActivateRegisteredMovementModeExtension (Variables|Default|GetLaunchSelfMME) (Ability|GetAvatarActorfromActorInfo)))))
        (:Cancelled
          _)
        (:Completed
          _))
      (CallFunction|HasAbilityStarted false)
      (Ability|EndAbility)
      (else
        (Ability|Tasks|PlayMontageandWaitforTrigger "None" _montage_to_play
          (:then
            (Variables|Default|SetbIsGroundJump true)
            (CallFunction|PauseUntilNextJump))
          (:Triggered
            _)
          (:Cancelled
            _)
          (:Completed
            _))))))

(event Custom|EndAbilityTimeout
  ; Wait until the player has landed before ending the ability- maybe needs some work to make it more robust?
  (Ability|EndAbility))

(event Custom|End%20Camera
  (Camera|PopCameraMode))

(event Custom|RemoveLaunchingTags
  (Ability|GameplayEffect|RemoveLooseGameplayTags (Variables|Default|GetPlayerPawn) (Variables|Default|GetTagsBlockBassWave) true))

(event Custom|PauseUntilNextJump
  ; After Bass Jumping, allow next Bass Jump after X duration (separate duration for first jump to allow for later ability application)
  (if (Variables|Default|GetbIsGroundJump)
    (Utilities|Time|SetTimerbyEvent (delegate Custom|AllowNextJump) (Default|ClampMinFloat (ScalableFloat|GetValueatLevel (Variables|Default|GetGround%20TIme%20Until%20Allow%20Next%20Jump)) (ScalableFloat|GetValueasDouble (Variables|Default|GetTime%20Until%20Allow%20Next%20Jump%20Minimum))))
    (else
      (Utilities|Time|SetTimerbyEvent (delegate Custom|AllowNextJump) (Default|ClampMinFloat (ScalableFloat|GetValueatLevel (Variables|Default|GetAerial%20Time%20Until%20Allow%20Next%20Jump)) (ScalableFloat|GetValueasDouble (Variables|Default|GetTime%20Until%20Allow%20Next%20Jump%20Minimum)))))))

(event Custom|AllowNextJump
  ; After Bass Jumping, allow next Bass Jump after X duration (separate duration for first jump to allow for later ability application)
  (CallFunction|HasAbilityStarted false)
  (Ability|EndAbility))

(event Ability|EventOnEndAbility
  (params (bWasCancelled :type Boolean))
  (CallFunction|HasAbilityStarted false)
  (CallFunction|RemoveLaunchingTags))

(fn BassJumpBlastThrough
  (bind (_returnvalue_x _returnvalue_y _returnvalue_z) (Math|Vector|BreakVector (Transformation|GetActorLocation (Variables|Default|GetPlayerPawn))))
  (bind (_returnvalue_x_2 _returnvalue_y_2 _returnvalue_z_2) (Math|Vector|BreakVector (Transformation|GetActorLocation (Variables|Default|GetPlayerPawn))))
  (if (> (ScalableFloat|GetValueatLevel (Variables|Default|GetBassJumpShouldDestroyAbove)) 0.0)
    ; Currently just applies vertically, investigate matching it to the player's path
    (bind _outhits (Collision|MultiCapsuleTraceForObjects :Start (Math|Vector|MakeVector _returnvalue_x_2 _returnvalue_y_2 (+ _returnvalue_z_2 (ScalableFloat|GetValueatLevel (Variables|Default|GetBassJumpDestroyDistanceZOffset)))) :End (Math|Vector|MakeVector _returnvalue_x _returnvalue_y (+ _returnvalue_z (ScalableFloat|GetValueatLevel (Variables|Default|GetBassJumpDestroyDistance)))) :Radius 8e+01 :HalfHeight 1.5e+02 :ObjectTypes (Variables|Default|GetObjectTypesToDestroy) :bTraceComplex false :DrawDebugType "Persistent" :bIgnoreSelf true :TraceColor "(R=1.000000,G=0.000000,B=0.000000,A=1.000000)" :TraceHitColor "(R=0.000000,G=1.000000,B=0.000000,A=1.000000)" :DrawTime 8.0))
    (for _array_element _outhits
      (if (Variables|/Script/Engine.Actor|GetbCanBeDamaged (.HitActor (Collision|BreakHitResult _array_element)))
        (bind _bdidactorhavetags (Class|AthenaQuestFunctionLibrary|AthenaCheckActorHasTags (.HitActor (Collision|BreakHitResult _array_element)) "(GameplayTags=((TagName=\"Deimos.Spawner\")))"))
        (if (not _bdidactorhavetags)
          (Utilities|Casting|CastToBuildingTrap (.HitActor (Collision|BreakHitResult _array_element))
            (:CastFailed
              (GameplayEffects|ApplyGameplayEffectToTarget :self (Ability|GetAbilitySystemComponent (Variables|Default|GetPlayerPawn)) :GameplayEffectClass "/Game/Athena/Items/Consumables/ShockwaveGrenade/GE_ShockGrenade_Damage.GE_ShockGrenade_Damage_C" :Target (Ability|GetAbilitySystemComponent (.HitActor (Collision|BreakHitResult _array_element))) :Level 0.0))))))))

(fn CanBassJump
  ; If player can multi-jump skip other checks
  (if (ScalableFloat|GetValueasBool (Variables|Default|GetbAllowMultiBassJump))
    (return true)
    (else
      (return (not (GameplayTags|HasMatchingGameplayTag (Variables|Default|GetPlayerPawn) (Variables|Default|GetTagPlayerBassJumpActive)))))))

(fn GetPolarityWeapon
  (bind _asb_weap_polarity_shield (Utilities|Casting|CastToB_Weap_PolarityShield (Ability|GetCurrentSourceWeapon))
    (:then
      (return _asb_weap_polarity_shield))
    (:CastFailed
      (return 0))))

(fn HasAbilityStarted
  (params (AbilityStart :type Boolean))
  (bind _asb_weap_polarity_shield (Utilities|Casting|CastToB_Weap_PolarityShield (Weapon|GetCurrentWeapon (Variables|Default|GetPlayerPawn))))
  (Variables|/GravWellGameplay/PolarityShield/B_Weap_PolarityShield.B_Weap_PolarityShield_C|SetLaunchActivated AbilityStart _asb_weap_polarity_shield))

(fn CanActivateAbility
  (params (ActorInfo :type Gameplay%20Ability%20Actor%20Info%20Structure) (Handle :type Gameplay%20Ability%20Spec%20Handle%20Structure))
  (return 0 (>= (ScalableFloat|GetValueatLevel (Variables|Default|GetHF_CanActivateAbility)) 1.0)))
