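; Gameplay ability for the Polarity Shield projectile's explosion in GravWell.
; On activation it fires an explosion cue ahead of the owning actor, then on authority runs separate targeting tasks for buildings and pawns/vehicles.
; ApplyGEsToTargets filters by immunity, shield-bubble tags and instigator, then sends knock events, checks building damage, applies vehicle or other effects and updates weapon damage analytics.
; A scalable flag enables explicit ending once both targeting passes finish.
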(event Ability|EventActivateAbility
  (Ability|ExecuteGameplayCueWithParamsOnOwner (Variables|Default|GetExplosionGameplayCueTag) (Ability|GameplayCue|MakeGameplayCueParameters :NormalizedMagnitude 0.0 :RawMagnitude 0.0 :Location (+ (Transformation|GetActorLocation (Ability|GetOwningActorfromActorInfo)) (* (Transformation|GetActorForwardVector (Ability|GetOwningActorfromActorInfo)) (Variables|Default|GetExplosionVFXForwardOffset))) :Normal (Transformation|GetActorForwardVector (Ability|GetOwningActorfromActorInfo)) :GameplayEffectLevel 1 :AbilityLevel 1 :bReplicateLocationWhenUsingMinimalRepProxy false))
  (Ability|HasAuthority
    (:True
      ; Apply GEs to Structures
      (bind (_asynctaskproxy _targetdata) (Ability|Tasks|PerformTargetingAbilityTask "None" "/GravWellGameplay/Gameplay/Targeting/TP_PolarityShield_ShieldExplosion.TP_PolarityShield_ShieldExplosion" false false true)
        (:then
          ; Apply GEs to Player/Boss Pawns & Vehicles
          (bind (_asynctaskproxy_2 _targetdata_2) (Ability|Tasks|PerformTargetingAbilityTask "None" "/GravWellGameplay/Gameplay/Targeting/TP_PolarityShield_ShieldExplosion_Pawns.TP_PolarityShield_ShieldExplosion_Pawns" false false true)
            (:Targeted
              (CallFunction|ApplyGEstoTargets (Variables|Default|GetPlayer%20And%20Boss%20Effects%20to%20Apply) (Ability|TargetData|GetAllActorsfromTargetData _targetdata_2)))
            (:Cancelled
              _))
          (if (ScalableFloat|GetValueasBool (Variables|Default|GetHF_ExplicitEndAbilityEnabled))
            (Variables|Default|SetbPawnTargetingFinished true)
            (CallFunction|TryEndAbility)))
        (:Targeted
          (CallFunction|ApplyGEstoTargets (Variables|Default|GetBuilding%20Effects%20to%20Apply) (Ability|TargetData|GetAllActorsfromTargetData _targetdata)))
        (:Cancelled
          _))
      (if (ScalableFloat|GetValueasBool (Variables|Default|GetHF_ExplicitEndAbilityEnabled))
        (Variables|Default|SetbBuildingTargetingFinished true)
        (CallFunction|TryEndAbility)))))

(event Ability|EventOnEndAbility
  (params (bWasCancelled :type Boolean)))

(fn ApplyGEsToTargets
  (params (EffectsToApply :type Array%20of%20Gameplay%20Effect%20Class%20References) (Actors :type Array%20of%20Actor%20Object%20References))
  (bind _returnvalue (Ability|GetOwningActorfromActorInfo))
  (bind _instigator (Variables|/Script/Engine.Actor|GetInstigator _returnvalue))
  (bind _returnvalue_2 (Ability|GetAbilitySystemComponent _instigator))
  (bind _returnvalue_3 (Class|AthenaFunctionLibrary|MakeSimpleFortEffectContext _instigator _returnvalue))
  (bind _returnvalue_4 (Transformation|GetActorLocation _returnvalue))
  (Variables|Default|SetEffects%20to%20Apply EffectsToApply)
  (for _array_element Actors
    (bind _output_get (Variables|Default|SetStruckActor _array_element))
    (if (GameplayTags|HasMatchingGameplayTag (Utilities|Casting|ToGameplayTagAssetInterface _output_get) "(TagName=\"GravWell.Explosion.Immunity\")")
      _
      (elif (GameplayTags|HasAnyMatchingGameplayTags (Utilities|Casting|ToGameplayTagAssetInterface (Variables|Default|GetStruckActor)) (Variables|Default|GetTagsInsideShieldBubble))
        (if (GameplayTags|HasMatchingGameplayTag (Utilities|Casting|ToGameplayTagAssetInterface (Ability|GetAvatarActorfromActorInfo)) "(TagName=\"GravWell.Shield.SpawnedInBubble\")")
          (if (!= (Variables|Default|GetStruckActor) _instigator)
            (CallFunction|SendQuestEvents (Variables|Default|GetStruckActor))
            (Class|BGAPolarityShieldShieldPrj|CheckforBuildingDamageDealt (Default|GetOwningProjectile) (Variables|Default|GetStruckActor) (ScalableFloat|GetValueatLevel (Variables|Default|GetExplosionBuildingDamage)) "Explosion")
            (bind _asfort_athena_vehicle (Utilities|Casting|CastToFortAthenaVehicle (Variables|Default|GetStruckActor))
              (:then
                (Fortnite|ApplyGameplayEffectsToActor _asfort_athena_vehicle (Variables|Default|GetVehicle%20Effects%20to%20Apply) _returnvalue_4 (ScalableFloat|GetValueasBool (Variables|Default|GetbRequireLOSForExplosionVehicle)) 0 _returnvalue_2 _returnvalue_3))
              (:CastFailed
                (Fortnite|ApplyGameplayEffectsToActor (Variables|Default|GetStruckActor) (Variables|Default|GetEffects%20to%20Apply) _returnvalue_4 (ScalableFloat|GetValueasBool (Variables|Default|GetbRequireLOSForExplosionNonVehicle)) 0 _returnvalue_2 _returnvalue_3)
                (bind _asgrav_well_weapon (Utilities|Casting|CastToGravWellWeapon (Variables|Default|GetPolaritySourceWeapon))
                  (:then
                    (Class|GravWellWeapon|UpdatePlayerDamageAnalytics _asgrav_well_weapon (ScalableFloat|GetValueasDouble (Variables|Default|GetExplosionPlayerDamage))))
                  (:CastFailed
                    (Development|PrintString "Unknown source weapon")))))))
        (else
          _)))))

(fn GetOwningProjectile
  (bind _asbga_polarity_shield_shield_prj (Utilities|Casting|CastToBGA_PolarityShield_ShieldPrj (Ability|GetOwningActorfromActorInfo))
    (:then
      (return _asbga_polarity_shield_shield_prj))
    (:CastFailed
      (return 0))))

(fn SendQuestEvents
  (params (AffectedActor :type Object%20Reference))
  (Utilities|Casting|CastToBuildingActor AffectedActor
    (:CastFailed
      (Utilities|Casting|CastToFortPawn AffectedActor
        (:then
          _)
        (:CastFailed
          (Utilities|Casting|CastToFortAthenaVehicle AffectedActor)))
      (Ability|SendGameplayEvent "(TagName=\"Event.Gravwell.Knock.Pawn\")" (Utilities|Struct|MakeGameplayEventData "(TagName=\"\")")))))

(fn SetPolarityWeaponRef
  (params (PolarityWeapon :type Fort%20Weapon%20Object%20Reference%20%28by%20ref%29))
  (Variables|Default|SetPolaritySourceWeapon PolarityWeapon))

(fn TryEndAbility
  (if (and (Variables|Default|GetbBuildingTargetingFinished) (Variables|Default|GetbPawnTargetingFinished))
    (Ability|EndAbility)))
