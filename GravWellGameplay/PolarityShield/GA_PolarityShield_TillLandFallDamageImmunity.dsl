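; Polarity Shield gameplay ability in GravWellGameplay, named for granting fall damage immunity until the player lands.
; Activation only calls the parent ability, so any immunity or landing logic is inherited or configured outside this graph.
; On end, cancelled or not, it calls the parent and then removes the player double-jumped loose gameplay tags from the avatar Fort player pawn.
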(event Ability|EventActivateAbility
  (|CallParentFunction))

(event Ability|EventOnEndAbility
  (params (bWasCancelled :type Boolean))
  (|CallParentFunction bWasCancelled)
  (Ability|GameplayEffect|RemoveLooseGameplayTags (CallFunction|GetAvatarasFortPlayerPawn) (Variables|Default|GetTagsPlayerDoubleJumped) true))
